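; BP_Papaya_Screen_01 is a Papaya video screen that plays a streamed movie with audio and subtitles.
; On remote clients it spawns a video stream trigger whose events drive material swaps, timeline aspect-ratio scaling, audio and subtitles.
; Before loading, it can delay playback until Screen_02 and Screen_03 stop.
; For share-locked sources, screen recording or content-protection settings gate video and audio, and audio submix routing switches on itself and Papaya speakers.
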(event Custom|EnableRedirectMaptoMusic
  (params (Enabled :type Boolean :default true))
  ;[1 1 1] MeshNetwork Control Redirectmap to Music Stage
  (if Enabled
    (Rendering|Material|SetMaterial (Variables|StaticMeshActor|GetStaticMeshComponent) 0 "/Game/Environments/Papaya/Materials/M_Papaya_Screen_RedirectMovie.M_Papaya_Screen_RedirectMovie")
    (Rendering|Material|SetScalarParameterValue "/Game/Athena/Events/Generic/CountdownTimer/Materials/MPC_Countdown.MPC_Countdown" "ShowPapayaRedirectShard" 1.0 :!class /Script/Engine.KismetMaterialLibrary)
    (else
      (Rendering|Material|SetMaterial (Variables|StaticMeshActor|GetStaticMeshComponent) 0 "/Game/Environments/Papaya/Materials/M_Papaya_Screen_copy.M_Papaya_Screen_copy")
      (Rendering|Material|SetScalarParameterValue "/Game/Athena/Events/Generic/CountdownTimer/Materials/MPC_Countdown.MPC_Countdown" "ShowPapayaRedirectShard" :!class /Script/Engine.KismetMaterialLibrary))))

(event Custom|SubtitleEnable
  (params (Enable :type Boolean) (FromStart :type Boolean))
  (bind _streamingtrigger (Variables|Default|GetStreamingTrigger))
  (bind _subtitleplayer (Variables|/Game/Athena/BuildingActors/Props/Building/ActorBlueprints/Containers/Athena_Prop_Video_Stream_Trigger1.Athena_Prop_Video_Stream_Trigger1_C|GetSubtitlePlayer _streamingtrigger))
  (if (or FromStart (Media|MediaPlayer|IsPlaying (Variables|Default|GetMediaPlayer)))
    (if (not (and (Math|Boolean|Equal(Boolean) (Variables|Default|GetEnableVideo) false) (Math|Boolean|Equal(Boolean) (Variables|Default|GetEnableAudio) false)))
      (if (Stream|ShouldStreambePlaying (Variables|Default|GetPlatformSource) (Variables|StaticMeshActor|GetStaticMeshComponent) 8e+03)
        (if (not (Variables|Default|GetSubtitlesEnabled))
          (Variables|Default|SetSubtitlesEnabled true)
          (FortniteSubtitles|SubtitlesPlayer|Play _subtitleplayer))
        (else
          _))
      (else
        _))
    (if (Variables|Default|GetSubtitlesEnabled)
      (Variables|Default|SetSubtitlesEnabled false)
      (FortniteSubtitles|SubtitlesPlayer|Stop _subtitleplayer))))

(event Custom|SharingCheck
  (CallFunction|UpdateBlocks)
  (CallFunction|VideoEnable (Variables|Default|GetEnableVideo))
  (CallFunction|AudioEnable (Variables|Default|GetEnableAudio))
  (Variables|Default|SetScreenRecordingLast (Variables|Default|GetScreenRecording))
  (CallFunction|SubtitleEnable false))

(event Custom|ResetScreenScale
  (bind _platformsource (Variables|Default|GetPlatformSource))
  (bind _aspectratio (Variables|/Script/FortniteGame.FortStreamMediaSource|GetAspectRatio _platformsource))
  (bind _16x9 (Variables|Default|Get16x9))
  (if (not (== _16x9 _aspectratio))
    (bind (_direction _newtrack_0) (Timeline|ResetScreenLerp PlayFromStart)
      (:Update
        (Transformation|SetRelativeScale3D (Variables|StaticMeshActor|GetStaticMeshComponent) (Math|Vector|MakeVector 1.0 1.0 (Math|Float|Lerp 1.0 (/ _16x9 _aspectratio) _newtrack_0))))
      (:Finished
        (Variables|/Script/FortniteGame.FortStreamMediaSource|SetAspectRatio (Variables|Default|Get16x9) (Variables|Default|GetPlatformSource))))))

(event Custom|VideoEnable
  (params (Enable :type Boolean) (Stopped :type Boolean))
  (if (Math|Boolean|NotEqual(Boolean) (Variables|Default|GetVideoEnabledLast) Enable)
    (if (Variables|/Script/FortniteGame.FortStreamMediaSource|GetbAudioOnly (Variables|Default|GetPlatformSource))
      _
      (elif Enable
        (Variables|Default|SetVideoEnabledLast true)
        (Rendering|Material|SetMaterial (Variables|StaticMeshActor|GetStaticMeshComponent) 0 "/Game/Movies/Season5/Materials/M_Papaya_VideoPlaying.M_Papaya_VideoPlaying")
        (Variables|/Game/Athena/BuildingActors/Props/Building/ActorBlueprints/Containers/Athena_Prop_Video_Stream_Trigger1.Athena_Prop_Video_Stream_Trigger1_C|SetScreenMesh (Variables|StaticMeshActor|GetStaticMeshComponent) (Variables|Default|GetStreamingTrigger))
        (else
          (Variables|Default|SetVideoEnabledLast false)
          (if (Variables|/Game/Athena/BuildingActors/Props/Building/ActorBlueprints/Containers/Athena_Prop_Video_Stream_Trigger1.Athena_Prop_Video_Stream_Trigger1_C|GetLeaveScreen (Variables|Default|GetStreamingTrigger))
            _
            (elif (not Stopped)
              (Rendering|Material|SetMaterial (Variables|StaticMeshActor|GetStaticMeshComponent) 0 "/Game/Environments/Papaya/Materials/M_Papaya_Screen_copy.M_Papaya_Screen_copy")
              (elif (and (Variables|Default|GetEventActive) (Variables|Default|GetEnableVideo))
                (Rendering|Material|SetMaterial (Variables|StaticMeshActor|GetStaticMeshComponent) 0 (Variables|Default|GetVideoStoppedMaterial))
                (else
                  (Rendering|Material|SetMaterial (Variables|StaticMeshActor|GetStaticMeshComponent) 0 "/Game/Environments/Papaya/Materials/M_Papaya_Screen_copy.M_Papaya_Screen_copy"))))))))))

(event Custom|SetScreenScale
  (bind _platformsource (Variables|Default|GetPlatformSource))
  (bind _aspectratio (Variables|/Script/FortniteGame.FortStreamMediaSource|GetAspectRatio _platformsource))
  (bind _16x9 (Variables|Default|Get16x9))
  (if (not (== _16x9 _aspectratio))
    (bind (_direction _newtrack_0) (Timeline|ScaleScreenLerp PlayFromStart)
      (:Update
        (Transformation|SetRelativeScale3D (Variables|StaticMeshActor|GetStaticMeshComponent) (Math|Vector|MakeVector 1.0 1.0 (Math|Float|Lerp 1.0 (/ _16x9 _aspectratio) _newtrack_0)))))))

(event Custom|AudioEnable
  (params (Enable :type Boolean))
  (bind _mediasound (Variables|Default|GetMediaSound))
  (if Enable
    (Components|Activation|Activate _mediasound false)
    (Audio|Components|Audio|SetVolumeMultiplier _mediasound 1.0 :!class /Script/AudioMixer.SynthComponent)
    (else
      (Audio|Components|Audio|SetVolumeMultiplier _mediasound 0.0 :!class /Script/AudioMixer.SynthComponent)
      (if (not (Variables|Default|GetEnableVideo))
        (Components|Activation|Deactivate _mediasound)))))

(event Custom|OnVideoStopped
  (CallFunction|VideoEnable false true)
  (CallFunction|ResetScreenScale)
  (Components|Activation|Deactivate (Variables|Default|GetMediaSound))
  (Variables|Default|SetStageScreenStopped false)
  (Variables|Default|SetStageSurroundScreenStopped false)
  (Variables|/Game/Athena/BuildingActors/Props/Building/ActorBlueprints/Containers/Athena_Prop_Video_Stream_Trigger1.Athena_Prop_Video_Stream_Trigger1_C|SetDelayVideoStart false (Variables|Default|GetStreamingTrigger)))

(event Custom|On%20Video%20Started
  (params (FirstStart :type Boolean))
  (CallFunction|FlipAudioSubmix true true)
  (CallFunction|UpdateBlocks)
  (CallFunction|SetScreenScale)
  (CallFunction|VideoEnable (Variables|Default|GetEnableVideo))
  (CallFunction|AudioEnable (Variables|Default|GetEnableAudio))
  (CallFunction|SubtitleEnable false true))

(event Custom|OnVideoAboutToStart
  (params (Reset :type Boolean))
  (bind _stagescreenhard (Variables|Default|GetStageScreenHard))
  (bind _streamingtrigger (Variables|/Game/Movies/Season5/Blueprints/BP_Papaya_Screen_02.BP_Papaya_Screen_02_C|GetStreamingTrigger _stagescreenhard))
  (bind _stagesurroundscreenshard (Variables|Default|GetStageSurroundScreensHard))
  (bind _streamingtrigger_2 (Variables|/Game/Movies/Season5/Blueprints/BP_Papaya_Screen_03.BP_Papaya_Screen_03_C|GetStreamingTrigger _stagesurroundscreenshard))
  (if (!= (Mesh|GetDataInt (Variables|Default|GetStreamingTrigger) (Variables|/Game/Athena/BuildingActors/Props/Building/ActorBlueprints/Containers/Athena_Prop_Video_Stream_Trigger1.Athena_Prop_Video_Stream_Trigger1_C|GetMesh_ScreenDistance (Variables|Default|GetStreamingTrigger))) 2)
    (if (not Reset)
      (Variables|/Game/Athena/BuildingActors/Props/Building/ActorBlueprints/Containers/Athena_Prop_Video_Stream_Trigger1.Athena_Prop_Video_Stream_Trigger1_C|SetDelayVideoStart true (Variables|Default|GetStreamingTrigger))
      (if (and (Media|MediaPlayer|IsPlaying (Variables|/Game/Athena/BuildingActors/Props/Building/ActorBlueprints/Containers/Athena_Prop_Video_Stream_Trigger1.Athena_Prop_Video_Stream_Trigger1_C|GetMediaPlayer _streamingtrigger)) (Variables|/Game/Athena/BuildingActors/Props/Building/ActorBlueprints/Containers/Athena_Prop_Video_Stream_Trigger1.Athena_Prop_Video_Stream_Trigger1_C|GetIsVideoTestEnabled _streamingtrigger))
        (Default|BindEventtoVideoStoppedEventDispatcher _streamingtrigger (delegate Custom|OtherScreenStopped))
        (else
          (CallFunction|OtherScreenStopped)))
      (if (and (Media|MediaPlayer|IsPlaying (Variables|/Game/Athena/BuildingActors/Props/Building/ActorBlueprints/Containers/Athena_Prop_Video_Stream_Trigger1.Athena_Prop_Video_Stream_Trigger1_C|GetMediaPlayer _streamingtrigger_2)) (Variables|/Game/Athena/BuildingActors/Props/Building/ActorBlueprints/Containers/Athena_Prop_Video_Stream_Trigger1.Athena_Prop_Video_Stream_Trigger1_C|GetIsVideoTestEnabled _streamingtrigger_2))
        (Default|BindEventtoVideoStoppedEventDispatcher _streamingtrigger_2 (delegate Custom|SurroundScreenStopped))
        (else
          (CallFunction|SurroundScreenStopped)))
      (else
        (Variables|/Game/Athena/BuildingActors/Props/Building/ActorBlueprints/Containers/Athena_Prop_Video_Stream_Trigger1.Athena_Prop_Video_Stream_Trigger1_C|SetDelayVideoStart false (Variables|Default|GetStreamingTrigger))
        (Default|UnbindEventfromVideoStoppedEventDispatcher (Variables|/Game/Movies/Season5/Blueprints/BP_Papaya_Screen_02.BP_Papaya_Screen_02_C|GetStreamingTrigger (Variables|Default|GetStageScreenHard)) (EventDispatchers|CreateEvent :!event Custom|OtherScreenStopped))
        (Default|UnbindEventfromVideoStoppedEventDispatcher (Variables|/Game/Movies/Season5/Blueprints/BP_Papaya_Screen_03.BP_Papaya_Screen_03_C|GetStreamingTrigger (Variables|Default|GetStageSurroundScreensHard)) (EventDispatchers|CreateEvent :!event Custom|SurroundScreenStopped))))))

(event EventBeginPlay
  (Components|Tick|SetComponentTickEnabled (Variables|Default|GetMediaSound) false)
  (Networking|SwitchHasAuthority
    (:Remote
      (Utilities|FlowControl|Sequence
        (:then_0
          (bind _returnvalue (Actor|GetActorOfClass "/Game/Movies/Season5/Blueprints/BP_Papaya_Screen_02.BP_Papaya_Screen_02_C"))
          (Utilities|IsValid _returnvalue
            (:Is%20Valid
              _)
            (:Is%20Not%20Valid
              (Utilities|FlowControl|Delay 0.25)))
          (Variables|Default|SetStageScreenHard _returnvalue)
          (Development|PrintString "GotScreen2"))
        (:then_1
          (bind _returnvalue_2 (Actor|GetActorOfClass "/Game/Movies/Season5/Blueprints/BP_Papaya_Screen_03.BP_Papaya_Screen_03_C"))
          (Utilities|IsValid _returnvalue_2
            (:Is%20Valid
              _)
            (:Is%20Not%20Valid
              (Utilities|FlowControl|Delay 0.25)))
          (Variables|Default|SetStageSurroundScreensHard _returnvalue_2)
          (Development|PrintString "Got Screen3"))
        (:then_2
          (Development|PrintString "Create stream trigger")
          (bind _returnvalue_3 (Game|SpawnActorfromClass "/Game/Athena/BuildingActors/Props/Building/ActorBlueprints/Containers/Athena_Prop_Video_Stream_Trigger1.Athena_Prop_Video_Stream_Trigger1_C" (Transformation|GetActorTransform) "Undefined" "OverrideRootScale" 0 (Variables|Default|GetMediaPlayer) (Variables|Default|GetPlatformSource) (Variables|StaticMeshActor|GetStaticMeshComponent) false "/Game/UI/Foundation/Movie/SubtitlePlayer_Papaya1.SubtitlePlayer_Papaya1" "/Game/Movies/Papaya_SRT_Electra1.Papaya_SRT_Electra1" "zVEuFPigBeWeAVTu" "Papaya1" 4e+04 1.5e+04 1e+04 false false self false 0 0))
          (bind _output_get (Variables|Default|SetStreamingTrigger _returnvalue_3))
          (Default|BindEventtoVideoStartedEventDispatcher _output_get (EventDispatchers|CreateEvent :!event Custom|On%20Video%20Started))
          (Default|BindEventtoVideoStoppedEventDispatcher _output_get (EventDispatchers|CreateEvent :!event Custom|OnVideoStopped))
          (Default|BindEventtoVideoAbouttoLoadEventDispatcher _output_get (EventDispatchers|CreateEvent :!event Custom|OnVideoAboutToStart))
          (Default|BindEventtoVideoSharingCheck _output_get (EventDispatchers|CreateEvent :!event Custom|SharingCheck)))))))

(event Custom|OtherScreenStopped
  (Variables|Default|SetStageScreenStopped true)
  (if (Variables|Default|GetStageSurroundScreenStopped)
    (Variables|/Game/Athena/BuildingActors/Props/Building/ActorBlueprints/Containers/Athena_Prop_Video_Stream_Trigger1.Athena_Prop_Video_Stream_Trigger1_C|SetDelayVideoStart false (Variables|Default|GetStreamingTrigger))))

(event Custom|SurroundScreenStopped
  (Variables|Default|SetStageSurroundScreenStopped true)
  (if (Variables|Default|GetStageScreenStopped)
    (Variables|/Game/Athena/BuildingActors/Props/Building/ActorBlueprints/Containers/Athena_Prop_Video_Stream_Trigger1.Athena_Prop_Video_Stream_Trigger1_C|SetDelayVideoStart false (Variables|Default|GetStreamingTrigger))))

(event EventEndPlay
  (params (EndPlayReason :type EEndPlayReason%20Enum :default "Destroyed"))
  (CallFunction|SubtitleEnable false)
  (EventDispatchers|UnbindallEventsfromApplicationHasEnteredForegroundDelegate (Variables|Default|GetApplicationLifecycle))
  (Actor|DestroyActor (Variables|Default|GetStreamingTrigger)))

(fn FlipAudioSubmix
  (params (First :type Boolean) (Default :type Boolean))
  (bind _mediasound (Variables|Default|GetMediaSound))
  (if (or First (Math|Boolean|NotEqual(Boolean) Default (Variables|Default|GetAudioSubmixDefault)))
    (if Default
      (Variables|Default|SetAudioSubmixDefault true)
      (Audio|Components|Audio|SetSubmixSend _mediasound "/Game/Sounds/SubMixes/Fortnite_LicensedContent_3D.Fortnite_LicensedContent_3D" 0.0 :!class /Script/AudioMixer.SynthComponent)
      (Audio|Components|Audio|SetSubmixSend _mediasound "/Game/Sounds/SubMixes/Fortnite_SFX.Fortnite_SFX" 1.0 :!class /Script/AudioMixer.SynthComponent)
      (else
        (Variables|Default|SetAudioSubmixDefault false)
        (Audio|Components|Audio|SetSubmixSend _mediasound "/Game/Sounds/SubMixes/Fortnite_SFX.Fortnite_SFX" 0.0 :!class /Script/AudioMixer.SynthComponent)
        (Audio|Components|Audio|SetSubmixSend _mediasound "/Game/Sounds/SubMixes/Fortnite_LicensedContent_3D.Fortnite_LicensedContent_3D" 1.0 :!class /Script/AudioMixer.SynthComponent)))
    (bind _outactors (Actor|GetAllActorsOfClass "/Game/Athena/Apollo/Environments/BuildingActors/Papaya/Speakers_Small_Papaya.Speakers_Small_Papaya_C"))
    (for _array_element _outactors
      (Utilities|IsValid (Variables|/Game/Athena/Apollo/Environments/BuildingActors/Papaya/Speakers_Small_Papaya.Speakers_Small_Papaya_C|GetPapayaPrototypeSpeakerComponent _array_element)
        (:Is%20Valid
          (Class|PapayaSpeakerComponent|SetSubmixSend (Variables|/Game/Athena/Apollo/Environments/BuildingActors/Papaya/Speakers_Small_Papaya.Speakers_Small_Papaya_C|GetPapayaPrototypeSpeakerComponent _array_element) Default))))))

(fn UpdateBlocks
  (bind _streamingtrigger (Variables|Default|GetStreamingTrigger))
  (bind _mediasource (Variables|/Game/Athena/BuildingActors/Props/Building/ActorBlueprints/Containers/Athena_Prop_Video_Stream_Trigger1.Athena_Prop_Video_Stream_Trigger1_C|GetMediaSource _streamingtrigger))
  (bind _returnvalue (Player|ShouldProtectPlayerfromContent _mediasource))
  (bind _returnvalue_2 (Utilities|Enum|Equal(Enum) _returnvalue "UCPBoth"))
  (bind _returnvalue_3 (not _returnvalue_2))
  (if (Variables|/Script/FortniteGame.FortStreamMediaSource|GetbShareLock _mediasource)
    (bind _returnvalue_4 (Player|IsScreenRecordinginProgress _mediasource))
    (bind _output_get (Variables|Default|SetScreenRecording _returnvalue_4))
    (if _returnvalue_4
      (Variables|Default|SetEnableVideo (not _output_get))
      (Variables|Default|SetEnableAudio (not _output_get))
      (else
        (Variables|Default|SetEnableVideo (and (not (Utilities|Enum|Equal(Enum) _returnvalue "UCPVideo")) _returnvalue_3))
        (Variables|Default|SetEnableAudio (and (not (Utilities|Enum|Equal(Enum) _returnvalue "UCPAudio")) _returnvalue_3))))
    (CallFunction|FlipAudioSubmix false (not (Utilities|String|Contains (Variables|/Script/FortniteGame.FortStreamMediaSource|GetProtectUserFromAVSettings (Variables|/Game/Athena/BuildingActors/Props/Building/ActorBlueprints/Containers/Athena_Prop_Video_Stream_Trigger1.Athena_Prop_Video_Stream_Trigger1_C|GetMediaSource (Variables|Default|GetStreamingTrigger))) "a")))))

(fn ConstructionScript)
