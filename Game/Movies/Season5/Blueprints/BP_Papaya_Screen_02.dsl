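; Second Papaya video screen, which on remote clients spawns a stream trigger to play streamed media with shard materials, audio and subtitles.
; Trigger events drive material swaps and audio, filtered by content protection and screen-recording checks that also reroute screen and speaker submixes.
; Video start can wait on the BP_Papaya_Screen_01 screen stopping.
; A region-specific replicated flag switches redirect-map materials, using timers to poll until region info populates, including join-in-progress.
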(event Custom|JIPSetMovieScreenRedirect
  (bind _streamingtrigger (Variables|Default|GetStreamingTrigger))
  ;[1 1 1] MeshNetwork Control Redirectmap to Movie Screen
  (Networking|CanExecuteCosmeticEvents
    (:True
      (bind _ispopulated (Class|AthenaPropVideoStreamTrigger1|IsRegionInfoPopulated (Variables|Default|GetStreamingTrigger)))
      (if _ispopulated
        (bind _out (Mesh|GetMeshDataBool _streamingtrigger (Variables|MeshVars|GetRedirectToMovieScreen%20GS) false))
        (CallFunction|EnableRedirectMaptoMovie (Mesh|GetDataBool _streamingtrigger _out))
        (elif (Utilities|Time|DoesTimerExistbyHandle (Variables|MeshVars|GetRedirectToMovieScreenHandle))
          _
          (else
            (bind _returnvalue (Utilities|Time|SetTimerbyEvent (EventDispatchers|CreateEvent :!event Custom|CheckMusicScreenRegionData) 0.01 true))
            (Variables|MeshVars|SetRedirectToMovieScreenHandleJIP _returnvalue)))))))

(event Custom|ShowRedirectToMovieScreen
  (params (Show :type Boolean) (Region%20Data :type String))
  ;[1 1 1] MeshNetwork Control Redirectmap to Movie Screen
  (Variables|MeshVars|SetRegionRedirectToMovieScreen (Utilities|Struct|MakeMeshRegionLockDataBool Show Region%20Data)))

(event Custom|CheckMusicScreenRegionData
  (bind _streamingtrigger (Variables|Default|GetStreamingTrigger))
  (bind _streamingtrigger_2 (Variables|Default|GetStreamingTrigger))
  ;[1 1 1] MeshNetwork Control Redirectmap to Movie Screen
  (Utilities|IsValid _streamingtrigger_2
    (:Is%20Valid
      (bind _ispopulated (Class|AthenaPropVideoStreamTrigger1|IsRegionInfoPopulated _streamingtrigger_2))
      (if _ispopulated
        (Utilities|Time|ClearandInvalidateTimerbyHandle (Variables|MeshVars|GetRedirectToMovieScreenHandle))
        (Utilities|Time|ClearandInvalidateTimerbyHandle (Variables|MeshVars|GetRedirectToMovieScreenHandleJIP))
        (bind _out (Mesh|GetMeshDataBool _streamingtrigger (Variables|MeshVars|GetRedirectToMovieScreen%20GS) false))
        (CallFunction|EnableRedirectMaptoMovie (Mesh|GetDataBool _streamingtrigger _out))))))

(event Custom|EnableRedirectMaptoMovie
  (params (Enable :type Boolean :default true))
  ;[1 1 1] MeshNetwork Control Redirectmap to Movie Screen
  (Networking|CanExecuteCosmeticEvents
    (:True
      (if Enable
        (Rendering|Material|SetMaterial (Variables|Default|GetShards) 0 "/Game/Environments/Papaya/Materials/MI_Shard_Papaya_Shards_Redirectmap.MI_Shard_Papaya_Shards_Redirectmap")
        (Rendering|Material|SetMaterial (Variables|StaticMeshActor|GetStaticMeshComponent) 0 "/Game/Environments/Papaya/Materials/MI_Shard_Papaya_Screen_Redirectmap.MI_Shard_Papaya_Screen_Redirectmap")
        (Rendering|Material|SetScalarParameterValue "/Game/Athena/Events/Generic/CountdownTimer/Materials/MPC_Countdown.MPC_Countdown" "ShowPapayaRedirectMovie" 1.0 :!class /Script/Engine.KismetMaterialLibrary)
        (else
          (Rendering|Material|SetMaterial (Variables|Default|GetShards) 0 "/Game/Environments/Papaya/Materials/MI_Shard_Papaya_Shards_Intermission.MI_Shard_Papaya_Shards_Intermission")
          (Rendering|Material|SetMaterial (Variables|StaticMeshActor|GetStaticMeshComponent) 0 "/Game/Environments/Papaya/Materials/MI_Shard_Papaya_Screen_Intermission.MI_Shard_Papaya_Screen_Intermission")
          (Rendering|Material|SetScalarParameterValue "/Game/Athena/Events/Generic/CountdownTimer/Materials/MPC_Countdown.MPC_Countdown" "ShowPapayaRedirectMovie" :!class /Script/Engine.KismetMaterialLibrary))))))

(event Custom|SubtitleEnable
  (params (Enable :type Boolean) (FromStart :type Boolean))
  (bind _streamingtrigger (Variables|Default|GetStreamingTrigger))
  (bind _subtitleplayer (Variables|/Game/Athena/BuildingActors/Props/Building/ActorBlueprints/Containers/Athena_Prop_Video_Stream_Trigger1.Athena_Prop_Video_Stream_Trigger1_C|GetSubtitlePlayer _streamingtrigger))
  (if (or FromStart (Media|MediaPlayer|IsPlaying (Variables|Default|GetMediaPlayer)))
    (if (not (and (Math|Boolean|Equal(Boolean) (Variables|Default|GetEnableVideo) false) (Math|Boolean|Equal(Boolean) (Variables|Default|GetEnableAudio) false)))
      (if (Stream|ShouldStreambePlaying (Variables|Default|GetPlatformSource) (Variables|StaticMeshActor|GetStaticMeshComponent) 1.2e+04)
        (if (not (Variables|Default|GetSubtitlesEnabled))
          (Variables|Default|SetSubtitlesEnabled true)
          (FortniteSubtitles|SubtitlesPlayer|Play _subtitleplayer))
        (else
          _))
      (else
        _))
    (if (Variables|Default|GetSubtitlesEnabled)
      (Variables|Default|SetSubtitlesEnabled false)
      (FortniteSubtitles|SubtitlesPlayer|Stop _subtitleplayer))))

(event Custom|HandleShards
  (params (Enable :type Boolean))
  (if Enable
    (Rendering|Material|SetMaterial (Variables|Default|GetShards) 0 "/Game/Environments/Papaya/Materials/MI_Shard_Papaya_Shards.MI_Shard_Papaya_Shards")
    (else
      (Rendering|Material|SetMaterial (Variables|Default|GetShards) 0 "/Game/Environments/Papaya/Materials/MI_Shard_Papaya_Shards_Intermission.MI_Shard_Papaya_Shards_Intermission"))))

(event Custom|SharingCheck
  (CallFunction|UpdateBlocks)
  (CallFunction|VideoEnable (Variables|Default|GetEnableVideo))
  (CallFunction|AudioEnable (Variables|Default|GetEnableAudio))
  (Variables|Default|SetScreenRecordingLast (Variables|Default|GetScreenRecording))
  (CallFunction|SubtitleEnable false))

(event Custom|VideoEnable
  (params (Enable :type Boolean) (Stopped :type Boolean))
  (if (Math|Boolean|NotEqual(Boolean) (Variables|Default|GetVideoEnabledLast) Enable)
    (if (Variables|/Script/FortniteGame.FortStreamMediaSource|GetbAudioOnly (Variables|Default|GetPlatformSource))
      _
      (elif Enable
        (Variables|Default|SetVideoEnabledLast true)
        (Rendering|Material|SetMaterial (Variables|StaticMeshActor|GetStaticMeshComponent) 0 "/Game/Environments/Papaya/Materials/MI_Shard_Papaya_Screen.MI_Shard_Papaya_Screen")
        (Variables|/Game/Athena/BuildingActors/Props/Building/ActorBlueprints/Containers/Athena_Prop_Video_Stream_Trigger1.Athena_Prop_Video_Stream_Trigger1_C|SetScreenMesh (Variables|StaticMeshActor|GetStaticMeshComponent) (Variables|Default|GetStreamingTrigger))
        (else
          (Variables|Default|SetVideoEnabledLast false)
          (if (Variables|/Game/Athena/BuildingActors/Props/Building/ActorBlueprints/Containers/Athena_Prop_Video_Stream_Trigger1.Athena_Prop_Video_Stream_Trigger1_C|GetLeaveScreen (Variables|Default|GetStreamingTrigger))
            _
            (elif (not Stopped)
              (Rendering|Material|SetMaterial (Variables|StaticMeshActor|GetStaticMeshComponent) 0 "/Game/Environments/Papaya/Materials/MI_Shard_Papaya_Shards_Intermission.MI_Shard_Papaya_Shards_Intermission")
              (elif (and (Variables|Default|GetEventActive) (Variables|Default|GetEnableVideo))
                (Rendering|Material|SetMaterial (Variables|StaticMeshActor|GetStaticMeshComponent) 0 (Variables|Default|GetVideoStoppedMaterial))
                (else
                  (Rendering|Material|SetMaterial (Variables|StaticMeshActor|GetStaticMeshComponent) 0 "/Game/Environments/Papaya/Materials/MI_Shard_Papaya_Shards_Intermission.MI_Shard_Papaya_Shards_Intermission"))))))))))

(event Custom|AudioEnable
  (params (Enable :type Boolean))
  (bind _mediasound (Variables|Default|GetMediaSound))
  (if Enable
    (Components|Activation|Activate _mediasound false)
    (Audio|Components|Audio|SetVolumeMultiplier _mediasound 1.0 :!class /Script/AudioMixer.SynthComponent)
    (else
      (Audio|Components|Audio|SetVolumeMultiplier _mediasound 0.0 :!class /Script/AudioMixer.SynthComponent)
      (if (not (Variables|Default|GetEnableVideo))
        (Components|Activation|Deactivate _mediasound)))))

(event Custom|OnVideoStopped
  (CallFunction|VideoEnable false true)
  (CallFunction|HandleShards false)
  (Components|Activation|Deactivate (Variables|Default|GetMediaSound)))

(event Custom|On%20Video%20Started
  (params (FirstStart :type Boolean))
  (CallFunction|FlipAudioSubmix true true)
  (CallFunction|UpdateBlocks)
  (CallFunction|VideoEnable (Variables|Default|GetEnableVideo))
  (CallFunction|AudioEnable (Variables|Default|GetEnableAudio))
  (CallFunction|HandleShards true)
  (CallFunction|SubtitleEnable false true))

(event Custom|OnVideoAboutToStart
  (params (Reset :type Boolean))
  (bind _theaterscreenhard (Variables|Default|GetTheaterScreenHard))
  (bind _streamingtrigger (Variables|/Game/Movies/Season5/Blueprints/BP_Papaya_Screen_01.BP_Papaya_Screen_01_C|GetStreamingTrigger _theaterscreenhard))
  (if (!= (Mesh|GetDataInt (Variables|Default|GetStreamingTrigger) (Variables|/Game/Athena/BuildingActors/Props/Building/ActorBlueprints/Containers/Athena_Prop_Video_Stream_Trigger1.Athena_Prop_Video_Stream_Trigger1_C|GetMesh_ScreenDistance (Variables|Default|GetStreamingTrigger))) 2)
    (if (not Reset)
      (Variables|/Game/Athena/BuildingActors/Props/Building/ActorBlueprints/Containers/Athena_Prop_Video_Stream_Trigger1.Athena_Prop_Video_Stream_Trigger1_C|SetDelayVideoStart true (Variables|Default|GetStreamingTrigger))
      (if (and (Media|MediaPlayer|IsPlaying (Variables|/Game/Athena/BuildingActors/Props/Building/ActorBlueprints/Containers/Athena_Prop_Video_Stream_Trigger1.Athena_Prop_Video_Stream_Trigger1_C|GetMediaPlayer _streamingtrigger)) (Variables|/Game/Athena/BuildingActors/Props/Building/ActorBlueprints/Containers/Athena_Prop_Video_Stream_Trigger1.Athena_Prop_Video_Stream_Trigger1_C|GetIsVideoTestEnabled _streamingtrigger))
        (Default|BindEventtoVideoStoppedEventDispatcher _streamingtrigger (delegate Custom|OtherScreenStopped))
        (else
          (CallFunction|OtherScreenStopped)))
      (else
        (Variables|/Game/Athena/BuildingActors/Props/Building/ActorBlueprints/Containers/Athena_Prop_Video_Stream_Trigger1.Athena_Prop_Video_Stream_Trigger1_C|SetDelayVideoStart false (Variables|Default|GetStreamingTrigger))
        (Default|UnbindEventfromVideoStoppedEventDispatcher (Variables|/Game/Movies/Season5/Blueprints/BP_Papaya_Screen_01.BP_Papaya_Screen_01_C|GetStreamingTrigger (Variables|Default|GetTheaterScreenHard)) (EventDispatchers|CreateEvent :!event Custom|OtherScreenStopped))))))

(event EventBeginPlay
  (Components|Tick|SetComponentTickEnabled (Variables|Default|GetMediaSound) false)
  (Networking|SwitchHasAuthority
    (:Remote
      (bind _returnvalue (Actor|GetActorOfClass "/Game/Movies/Season5/Blueprints/BP_Papaya_Screen_01.BP_Papaya_Screen_01_C"))
      (Utilities|IsValid _returnvalue
        (:Is%20Valid
          _)
        (:Is%20Not%20Valid
          (Utilities|FlowControl|Delay 0.25)))
      (Variables|Default|SetTheaterScreenHard _returnvalue)
      (Development|PrintString "Create stream trigger")
      (bind _returnvalue_2 (Game|SpawnActorfromClass "/Game/Athena/BuildingActors/Props/Building/ActorBlueprints/Containers/Athena_Prop_Video_Stream_Trigger1.Athena_Prop_Video_Stream_Trigger1_C" (Transformation|GetActorTransform) "Undefined" "OverrideRootScale" 0 (Variables|Default|GetMediaPlayer) (Variables|Default|GetPlatformSource) (Variables|StaticMeshActor|GetStaticMeshComponent) false "/Game/UI/Foundation/Movie/SubtitlePlayer_Papaya2.SubtitlePlayer_Papaya2" "/Game/Movies/Papaya_SRT_Electra2.Papaya_SRT_Electra2" "fTpbbVJbTHfGrFnH" "Papaya2" 4e+04 1.2e+04 1.5e+04 false false self false 0 0))
      (bind _output_get (Variables|Default|SetStreamingTrigger _returnvalue_2))
      (Default|BindEventtoVideoStartedEventDispatcher _output_get (EventDispatchers|CreateEvent :!event Custom|On%20Video%20Started))
      (Default|BindEventtoVideoStoppedEventDispatcher _output_get (EventDispatchers|CreateEvent :!event Custom|OnVideoStopped))
      (Default|BindEventtoVideoAbouttoLoadEventDispatcher _output_get (EventDispatchers|CreateEvent :!event Custom|OnVideoAboutToStart))
      (Default|BindEventtoVideoSharingCheck _output_get (EventDispatchers|CreateEvent :!event Custom|SharingCheck))
      (CallFunction|JIPSetMovieScreenRedirect))))

(event Custom|OtherScreenStopped
  (Variables|/Game/Athena/BuildingActors/Props/Building/ActorBlueprints/Containers/Athena_Prop_Video_Stream_Trigger1.Athena_Prop_Video_Stream_Trigger1_C|SetDelayVideoStart false (Variables|Default|GetStreamingTrigger)))

(event EventEndPlay
  (params (EndPlayReason :type EEndPlayReason%20Enum :default "Destroyed"))
  (CallFunction|SubtitleEnable false)
  (EventDispatchers|UnbindallEventsfromApplicationHasEnteredForegroundDelegate (Variables|Default|GetApplicationLifecycle))
  (Actor|DestroyActor (Variables|Default|GetStreamingTrigger)))

(fn FlipAudioSubmix
  (params (Force :type Boolean) (Default :type Boolean))
  (bind _mediasound (Variables|Default|GetMediaSound))
  (if (or (Math|Boolean|NotEqual(Boolean) Default (Variables|Default|GetAudioSubmixDefault)) Force)
    (if Default
      (Variables|Default|SetAudioSubmixDefault true)
      (Audio|Components|Audio|SetSubmixSend _mediasound "/Game/Sounds/SubMixes/Fortnite_LicensedContent_3D.Fortnite_LicensedContent_3D" 0.0 :!class /Script/AudioMixer.SynthComponent)
      (Audio|Components|Audio|SetSubmixSend _mediasound "/Game/Sounds/SubMixes/Fortnite_SFX.Fortnite_SFX" 1.0 :!class /Script/AudioMixer.SynthComponent)
      (else
        (Variables|Default|SetAudioSubmixDefault false)
        (Audio|Components|Audio|SetSubmixSend _mediasound "/Game/Sounds/SubMixes/Fortnite_SFX.Fortnite_SFX" 0.0 :!class /Script/AudioMixer.SynthComponent)
        (Audio|Components|Audio|SetSubmixSend _mediasound "/Game/Sounds/SubMixes/Fortnite_LicensedContent_3D.Fortnite_LicensedContent_3D" 1.0 :!class /Script/AudioMixer.SynthComponent)))
    (bind _outactors (Actor|GetAllActorsOfClass "/Game/Athena/Apollo/Environments/BuildingActors/Papaya/Speakers_Small_Papaya.Speakers_Small_Papaya_C"))
    (for _array_element _outactors
      (Utilities|IsValid (Variables|/Game/Athena/Apollo/Environments/BuildingActors/Papaya/Speakers_Small_Papaya.Speakers_Small_Papaya_C|GetPapayaPrototypeSpeakerComponent _array_element)
        (:Is%20Valid
          (Class|PapayaSpeakerComponent|SetSubmixSend (Variables|/Game/Athena/Apollo/Environments/BuildingActors/Papaya/Speakers_Small_Papaya.Speakers_Small_Papaya_C|GetPapayaPrototypeSpeakerComponent _array_element) Default))))))

(fn OnRep_RegionRedirectToMovieScreen
  (bind (_data _regiondata) (Utilities|Struct|BreakMeshRegionLockDataBool (Variables|MeshVars|GetRegionRedirectToMovieScreen)))
  (bind _streamingtrigger (Variables|Default|GetStreamingTrigger))
  (Class|FortMediaRegionLockComponent|MeshAddRegionValueBool _data (Variables|MeshVars|GetRedirectToMovieScreen%20GS) _regiondata)
  (Networking|CanExecuteCosmeticEvents
    (:True
      (Utilities|IsValid _streamingtrigger
        (:Is%20Valid
          (bind _ispopulated (Class|AthenaPropVideoStreamTrigger1|IsRegionInfoPopulated _streamingtrigger))
          (if _ispopulated
            (CallFunction|CheckMusicScreenRegionData)
            (else
              (bind _returnvalue (Utilities|Time|SetTimerbyEvent (EventDispatchers|CreateEvent :!event CheckMusicScreenRegionData) 0.01 true))
              (Variables|MeshVars|SetRedirectToMovieScreenHandle _returnvalue))))))))

(fn UpdateBlocks
  (bind _streamingtrigger (Variables|Default|GetStreamingTrigger))
  (bind _mediasource (Variables|/Game/Athena/BuildingActors/Props/Building/ActorBlueprints/Containers/Athena_Prop_Video_Stream_Trigger1.Athena_Prop_Video_Stream_Trigger1_C|GetMediaSource _streamingtrigger))
  (bind _returnvalue (Player|ShouldProtectPlayerfromContent _mediasource))
  (bind _returnvalue_2 (Utilities|Enum|Equal(Enum) _returnvalue "UCPBoth"))
  (bind _returnvalue_3 (not _returnvalue_2))
  (if (Variables|/Script/FortniteGame.FortStreamMediaSource|GetbShareLock _mediasource)
    (bind _returnvalue_4 (Player|IsScreenRecordinginProgress _mediasource))
    (bind _output_get (Variables|Default|SetScreenRecording _returnvalue_4))
    (if _returnvalue_4
      (Variables|Default|SetEnableVideo (not _output_get))
      (Variables|Default|SetEnableAudio (not _output_get))
      (else
        (Variables|Default|SetEnableVideo (and (not (Utilities|Enum|Equal(Enum) _returnvalue "UCPVideo")) _returnvalue_3))
        (Variables|Default|SetEnableAudio (and (not (Utilities|Enum|Equal(Enum) _returnvalue "UCPAudio")) _returnvalue_3))))
    (CallFunction|FlipAudioSubmix false (not (Utilities|String|Contains (Variables|/Script/FortniteGame.FortStreamMediaSource|GetProtectUserFromAVSettings (Variables|/Game/Athena/BuildingActors/Props/Building/ActorBlueprints/Containers/Athena_Prop_Video_Stream_Trigger1.Athena_Prop_Video_Stream_Trigger1_C|GetMediaSource (Variables|Default|GetStreamingTrigger))) "a")))))

(fn ConstructionScript)
