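; Papaya side shard screen that plays streamed video with audio and subtitles, hidden by default.
; Clients spawn a stream trigger whose events swap screen and shard materials and can delay playback until the BP_Papaya_Screen_01 theater video stops.
; Content protection and screen-recording checks gate video and audio and switch audio submixes.
; Mesh network states, applied through OnReps and startup polling timers, control visibility and default, mirror or split layouts.
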(event Custom|EnableRedirectMaptoMovie
  (params (Enable :type Boolean :default true))
  ;[1 1 1] MeshNetwork Control Redirectmap to Movie Screen
  (if Enable
    (Rendering|Material|SetMaterial (Variables|Default|GetShards) 0 "/Game/Environments/Papaya/Materials/MI_Shard_Papaya_Shards_Redirectmap.MI_Shard_Papaya_Shards_Redirectmap")
    (Rendering|Material|SetMaterial (Variables|StaticMeshActor|GetStaticMeshComponent) 0 "/Game/Environments/Papaya/Materials/MI_Shard_Papaya_Screen_Redirectmap.MI_Shard_Papaya_Screen_Redirectmap")
    (Rendering|Material|SetMaterial (Variables|StaticMeshActor|GetStaticMeshComponent) 1 "/Game/Environments/Papaya/Materials/MI_Shard_Papaya_Screen_Redirectmap.MI_Shard_Papaya_Screen_Redirectmap")
    (Rendering|Material|SetScalarParameterValue "/Game/Athena/Events/Generic/CountdownTimer/Materials/MPC_Countdown.MPC_Countdown" "ShowPapayaRedirectMovie" 1.0 :!class /Script/Engine.KismetMaterialLibrary)
    (else
      (Rendering|Material|SetMaterial (Variables|Default|GetShards) 0 "/Game/Environments/Papaya/Materials/MI_Shard_Papaya_Shards_Intermission.MI_Shard_Papaya_Shards_Intermission")
      (Rendering|Material|SetMaterial (Variables|StaticMeshActor|GetStaticMeshComponent) 0 "/Game/Environments/Papaya/Materials/MI_Shard_Papaya_Screen_Intermission.MI_Shard_Papaya_Screen_Intermission")
      (Rendering|Material|SetMaterial (Variables|StaticMeshActor|GetStaticMeshComponent) 1 "/Game/Environments/Papaya/Materials/MI_Shard_Papaya_Screen_Intermission.MI_Shard_Papaya_Screen_Intermission")
      (Rendering|Material|SetScalarParameterValue "/Game/Athena/Events/Generic/CountdownTimer/Materials/MPC_Countdown.MPC_Countdown" "ShowPapayaRedirectMovie" :!class /Script/Engine.KismetMaterialLibrary))))

(event Custom|SubtitleEnable
  (params (Enable :type Boolean) (FromStart :type Boolean))
  (bind _streamingtrigger (Variables|Default|GetStreamingTrigger))
  (bind _subtitleplayer (Variables|/Game/Athena/BuildingActors/Props/Building/ActorBlueprints/Containers/Athena_Prop_Video_Stream_Trigger1.Athena_Prop_Video_Stream_Trigger1_C|GetSubtitlePlayer _streamingtrigger))
  (if (or FromStart (Media|MediaPlayer|IsPlaying (Variables|Default|GetMediaPlayer)))
    (if (not (and (Math|Boolean|Equal(Boolean) (Variables|Default|GetEnableVideo) false) (Math|Boolean|Equal(Boolean) (Variables|Default|GetEnableAudio) false)))
      (if (Stream|ShouldStreambePlaying (Variables|Default|GetPlatformSource) (Variables|StaticMeshActor|GetStaticMeshComponent) 1.2e+04)
        (if (not (Variables|Default|GetSubtitlesEnabled))
          (Variables|Default|SetSubtitlesEnabled true)
          (FortniteSubtitles|SubtitlesPlayer|Play _subtitleplayer))
        (else
          _))
      (else
        _))
    (if (Variables|Default|GetSubtitlesEnabled)
      (Variables|Default|SetSubtitlesEnabled false)
      (FortniteSubtitles|SubtitlesPlayer|Stop _subtitleplayer))))

(event Custom|HandleShards
  (params (Enable :type Boolean))
  (if Enable
    (Rendering|Material|SetMaterial (Variables|Default|GetShards) 0 "/Game/Environments/Papaya/Materials/MI_Shard_Papaya_Shards.MI_Shard_Papaya_Shards")
    (else
      (Rendering|Material|SetMaterial (Variables|Default|GetShards) 0 "/Game/Environments/Papaya/Materials/MI_Shard_Papaya_Shards_Intermission.MI_Shard_Papaya_Shards_Intermission"))))

(event Custom|SharingCheck
  (CallFunction|UpdateBlocks)
  (CallFunction|VideoEnable (Variables|Default|GetEnableVideo))
  (CallFunction|AudioEnable (Variables|Default|GetEnableAudio))
  (Variables|Default|SetScreenRecordingLast (Variables|Default|GetScreenRecording))
  (CallFunction|SubtitleEnable false))

(event Custom|VideoEnable
  (params (Enable :type Boolean) (Stopped :type Boolean))
  (if (Math|Boolean|NotEqual(Boolean) (Variables|Default|GetVideoEnabledLast) Enable)
    (if (Variables|/Script/FortniteGame.FortStreamMediaSource|GetbAudioOnly (Variables|Default|GetPlatformSource))
      _
      (elif Enable
        (Variables|Default|SetVideoEnabledLast true)
        (Rendering|Material|SetMaterial (Variables|StaticMeshActor|GetStaticMeshComponent) 0 "/Game/Environments/Papaya/Materials/MI_Shard_Papaya_Screen2.MI_Shard_Papaya_Screen2")
        (Rendering|Material|SetMaterial (Variables|StaticMeshActor|GetStaticMeshComponent) 1 "/Game/Environments/Papaya/Materials/MI_Shard_Papaya_Screen2.MI_Shard_Papaya_Screen2")
        (Variables|/Game/Athena/BuildingActors/Props/Building/ActorBlueprints/Containers/Athena_Prop_Video_Stream_Trigger1.Athena_Prop_Video_Stream_Trigger1_C|SetScreenMesh (Variables|StaticMeshActor|GetStaticMeshComponent) (Variables|Default|GetStreamingTrigger))
        (else
          (Variables|Default|SetVideoEnabledLast false)
          (if (Variables|/Game/Athena/BuildingActors/Props/Building/ActorBlueprints/Containers/Athena_Prop_Video_Stream_Trigger1.Athena_Prop_Video_Stream_Trigger1_C|GetLeaveScreen (Variables|Default|GetStreamingTrigger))
            _
            (elif (not Stopped)
              (Rendering|Material|SetMaterial (Variables|StaticMeshActor|GetStaticMeshComponent) 0 "/Game/Environments/Papaya/Materials/MI_Shard_Papaya_Shards_Intermission.MI_Shard_Papaya_Shards_Intermission")
              (Rendering|Material|SetMaterial (Variables|StaticMeshActor|GetStaticMeshComponent) 1 "/Game/Environments/Papaya/Materials/MI_Shard_Papaya_Shards_Intermission.MI_Shard_Papaya_Shards_Intermission")
              (elif (and (Variables|Default|GetEventActive) (Variables|Default|GetEnableVideo))
                (Rendering|Material|SetMaterial (Variables|StaticMeshActor|GetStaticMeshComponent) 0 (Variables|Default|GetVideoStoppedMaterial))
                (Rendering|Material|SetMaterial (Variables|StaticMeshActor|GetStaticMeshComponent) 1 (Variables|Default|GetVideoStoppedMaterial))
                (else
                  (Rendering|Material|SetMaterial (Variables|StaticMeshActor|GetStaticMeshComponent) 0 "/Game/Environments/Papaya/Materials/MI_Shard_Papaya_Shards_Intermission.MI_Shard_Papaya_Shards_Intermission")
                  (Rendering|Material|SetMaterial (Variables|StaticMeshActor|GetStaticMeshComponent) 1 "/Game/Environments/Papaya/Materials/MI_Shard_Papaya_Shards_Intermission.MI_Shard_Papaya_Shards_Intermission"))))))))))

(event Custom|AudioEnable
  (params (Enable :type Boolean))
  (bind _mediasound (Variables|Default|GetMediaSound))
  (if Enable
    (Components|Activation|Activate _mediasound false)
    (Audio|Components|Audio|SetVolumeMultiplier _mediasound 1.0 :!class /Script/AudioMixer.SynthComponent)
    (else
      (Audio|Components|Audio|SetVolumeMultiplier _mediasound 0.0 :!class /Script/AudioMixer.SynthComponent)
      (if (not (Variables|Default|GetEnableVideo))
        (Components|Activation|Deactivate _mediasound)))))

(event Custom|OnVideoStopped
  (CallFunction|VideoEnable false true)
  (CallFunction|HandleShards false)
  (Components|Activation|Deactivate (Variables|Default|GetMediaSound)))

(event Custom|On%20Video%20Started
  (params (FirstStart :type Boolean))
  (CallFunction|FlipAudioSubmix true true)
  (CallFunction|UpdateBlocks)
  (CallFunction|VideoEnable (Variables|Default|GetEnableVideo))
  (CallFunction|AudioEnable (Variables|Default|GetEnableAudio))
  (CallFunction|HandleShards true)
  (CallFunction|SubtitleEnable false true)
  (CallFunction|HandleMesh))

(event EventBeginPlay
  (Components|Tick|SetComponentTickEnabled (Variables|Default|GetMediaSound) false)
  (Networking|SwitchHasAuthority
    (:Remote
      (bind _returnvalue (Actor|GetActorOfClass "/Game/Movies/Season5/Blueprints/BP_Papaya_Screen_01.BP_Papaya_Screen_01_C"))
      (Utilities|IsValid _returnvalue
        (:Is%20Valid
          _)
        (:Is%20Not%20Valid
          (Utilities|FlowControl|Delay 0.25)))
      (Variables|Default|SetTheaterScreenHard _returnvalue)
      (Development|PrintString "Create stream trigger")
      (bind _returnvalue_2 (Game|SpawnActorfromClass "/Game/Athena/BuildingActors/Props/Building/ActorBlueprints/Containers/Athena_Prop_Video_Stream_Trigger1.Athena_Prop_Video_Stream_Trigger1_C" (Transformation|GetActorTransform) "Undefined" "OverrideRootScale" 0 (Variables|Default|GetMediaPlayer) (Variables|Default|GetPlatformSource) (Variables|StaticMeshActor|GetStaticMeshComponent) false "/Game/UI/Foundation/Movie/SubtitlePlayer_Papaya2.SubtitlePlayer_Papaya2" "/Game/Movies/Papaya_SRT_Electra2.Papaya_SRT_Electra2" "hwlkMzGogRHdaqH" "Papaya3" 4e+04 1.2e+04 1.5e+04 false false self false 0 0))
      (bind _output_get (Variables|Default|SetStreamingTrigger _returnvalue_2))
      (Default|BindEventtoVideoStartedEventDispatcher _output_get (EventDispatchers|CreateEvent :!event Custom|On%20Video%20Started))
      (Default|BindEventtoVideoStoppedEventDispatcher _output_get (EventDispatchers|CreateEvent :!event Custom|OnVideoStopped))
      (Default|BindEventtoVideoAbouttoLoadEventDispatcher _output_get (EventDispatchers|CreateEvent :!event Custom|OnVideoAboutToStart))
      (Default|BindEventtoVideoSharingCheck _output_get (EventDispatchers|CreateEvent :!event Custom|SharingCheck)))))

(event Custom|OnVideoAboutToStart
  (params (Reset :type Boolean))
  (bind _theaterscreenhard (Variables|Default|GetTheaterScreenHard))
  (bind _streamingtrigger (Variables|/Game/Movies/Season5/Blueprints/BP_Papaya_Screen_01.BP_Papaya_Screen_01_C|GetStreamingTrigger _theaterscreenhard))
  (if (!= (Mesh|GetDataInt (Variables|Default|GetStreamingTrigger) (Variables|/Game/Athena/BuildingActors/Props/Building/ActorBlueprints/Containers/Athena_Prop_Video_Stream_Trigger1.Athena_Prop_Video_Stream_Trigger1_C|GetMesh_ScreenDistance (Variables|Default|GetStreamingTrigger))) 2)
    (if (not Reset)
      (Variables|/Game/Athena/BuildingActors/Props/Building/ActorBlueprints/Containers/Athena_Prop_Video_Stream_Trigger1.Athena_Prop_Video_Stream_Trigger1_C|SetDelayVideoStart true (Variables|Default|GetStreamingTrigger))
      (if (and (Media|MediaPlayer|IsPlaying (Variables|/Game/Athena/BuildingActors/Props/Building/ActorBlueprints/Containers/Athena_Prop_Video_Stream_Trigger1.Athena_Prop_Video_Stream_Trigger1_C|GetMediaPlayer _streamingtrigger)) (Variables|/Game/Athena/BuildingActors/Props/Building/ActorBlueprints/Containers/Athena_Prop_Video_Stream_Trigger1.Athena_Prop_Video_Stream_Trigger1_C|GetIsVideoTestEnabled _streamingtrigger))
        (Default|BindEventtoVideoStoppedEventDispatcher _streamingtrigger (delegate Custom|OtherScreenStopped))
        (else
          (CallFunction|OtherScreenStopped)))
      (else
        (Variables|/Game/Athena/BuildingActors/Props/Building/ActorBlueprints/Containers/Athena_Prop_Video_Stream_Trigger1.Athena_Prop_Video_Stream_Trigger1_C|SetDelayVideoStart false (Variables|Default|GetStreamingTrigger))
        (Default|UnbindEventfromVideoStoppedEventDispatcher (Variables|/Game/Movies/Season5/Blueprints/BP_Papaya_Screen_01.BP_Papaya_Screen_01_C|GetStreamingTrigger (Variables|Default|GetTheaterScreenHard)) (EventDispatchers|CreateEvent :!event Custom|OtherScreenStopped))))))

(event Custom|OtherScreenStopped
  (Variables|/Game/Athena/BuildingActors/Props/Building/ActorBlueprints/Containers/Athena_Prop_Video_Stream_Trigger1.Athena_Prop_Video_Stream_Trigger1_C|SetDelayVideoStart false (Variables|Default|GetStreamingTrigger)))

(event EventEndPlay
  (params (EndPlayReason :type EEndPlayReason%20Enum :default "Destroyed"))
  (CallFunction|SubtitleEnable false)
  (EventDispatchers|UnbindallEventsfromApplicationHasEnteredForegroundDelegate (Variables|Default|GetApplicationLifecycle))
  (Actor|DestroyActor (Variables|Default|GetStreamingTrigger)))

(event Custom|HandleMesh
  (CallFunction|DefaultSideScreensStartup)
  (CallFunction|MirrorSideScreensStartup)
  (CallFunction|ShowShardSideScreensStartup)
  (CallFunction|SplitSideScreensStartup))

(event Custom|DefaultSideScreens
  (params (Enable :type Boolean))
  (Default|MeshNode
    (:Client
      (if (Variables|/Script/FortniteGame.FortStreamMediaSource|GetbAudioOnly (Variables|Default|GetPlatformSource))
        _
        (elif (Variables|Default|GetEnableVideo)
          (Rendering|Material|SetMaterial (Variables|StaticMeshActor|GetStaticMeshComponent) 0 "/Game/Environments/Papaya/Materials/MI_Shard_Papaya_Screen2.MI_Shard_Papaya_Screen2")
          (Rendering|Material|SetMaterial (Variables|StaticMeshActor|GetStaticMeshComponent) 1 "/Game/Environments/Papaya/Materials/MI_Shard_Papaya_Screen2.MI_Shard_Papaya_Screen2"))))
    (:Root
      (Variables|Mesh|SetDefaultSideScreens_MState (Math|Integer|SelectInt 1 0 Enable))
      (Variables|Mesh|SetMirrorSideScreens_MState 0)
      (Variables|Mesh|SetSplitSideScreens_MState 0))))

(event Custom|MirrorSideScreens
  (params (Enable :type Boolean))
  (bind _returnvalue (Math|Integer|SelectInt 1 0 Enable))
  (Default|MeshNode
    (:Client
      (if (Variables|/Script/FortniteGame.FortStreamMediaSource|GetbAudioOnly (Variables|Default|GetPlatformSource))
        _
        (elif (Variables|Default|GetEnableVideo)
          (Utilities|MeshEnabled _returnvalue
            (:Enabled
              (Rendering|Material|SetMaterial (Variables|StaticMeshActor|GetStaticMeshComponent) 0 "/Game/Environments/Papaya/Materials/MI_Shard_Papaya_Screen2.MI_Shard_Papaya_Screen2")
              (Rendering|Material|SetMaterial (Variables|StaticMeshActor|GetStaticMeshComponent) 1 "/Game/Environments/Papaya/Materials/MI_Shard_Papaya_Screen_RightMirror.MI_Shard_Papaya_Screen_RightMirror"))
            (:Disabled
              (CallFunction|DefaultSideScreens true))))))
    (:Root
      (Variables|Mesh|SetMirrorSideScreens_MState _returnvalue))))

(event Custom|SplitSideScreens
  (params (Enable :type Boolean))
  (bind _returnvalue (Math|Integer|SelectInt 1 0 Enable))
  (Default|MeshNode
    (:Client
      (if (Variables|/Script/FortniteGame.FortStreamMediaSource|GetbAudioOnly (Variables|Default|GetPlatformSource))
        _
        (elif (Variables|Default|GetEnableVideo)
          (Utilities|MeshEnabled _returnvalue
            (:Enabled
              (Rendering|Material|SetMaterial (Variables|StaticMeshActor|GetStaticMeshComponent) 0 "/Game/Environments/Papaya/Materials/MI_Shard_Papaya_Screen_LeftSplit.MI_Shard_Papaya_Screen_LeftSplit")
              (Rendering|Material|SetMaterial (Variables|StaticMeshActor|GetStaticMeshComponent) 1 "/Game/Environments/Papaya/Materials/MI_Shard_Papaya_Screen_RightSplit.MI_Shard_Papaya_Screen_RightSplit"))
            (:Disabled
              (CallFunction|DefaultSideScreens true))))))
    (:Root
      (Variables|Mesh|SetSplitSideScreens_MState _returnvalue))))

(event Custom|ShowShardSideScreens
  (params (Enable :type Boolean))
  (bind _returnvalue (Math|Integer|SelectInt 1 0 Enable))
  (Default|MeshNode
    (:Client
      (Utilities|MeshEnabled _returnvalue
        (:Enabled
          (Development|SetHiddeninGame (Variables|StaticMeshActor|GetStaticMeshComponent) false false))
        (:Disabled
          (Development|SetHiddeninGame (Variables|StaticMeshActor|GetStaticMeshComponent) true false))))
    (:Root
      (Variables|Mesh|SetShowShardSideScreens_MState _returnvalue))))

(event Custom|ShowShardSideScreens_Startup
  (Utilities|MeshEnabled (Variables|Mesh|GetShowShardSideScreens_MState)
    (:Enabled
      (CallFunction|ShowShardSideScreens true))
    (:Disabled
      (CallFunction|ShowShardSideScreens false))
    (:Default
      (bind _returnvalue (Utilities|Time|SetTimerbyEvent (delegate Custom|HandleShowSideScreens) 0.5 true))
      (Variables|Mesh|SetShowShardSideScreens_TH _returnvalue)
      (Variables|Mesh|SetShowShardSideScreens_TimeTracker (Math|DateTime|UTCNow)))))

(event Custom|HandleShowSideScreens
  (Utilities|MeshEnabled (Variables|Mesh|GetShowShardSideScreens_MState)
    (:Enabled
      _)
    (:Disabled
      _)
    (:Default
      (Utilities|OutOfTime (Variables|Mesh|GetShowShardSideScreens_TimeTracker)
        (:OutOfTime
          (Utilities|Time|ClearandInvalidateTimerbyHandle (Variables|Mesh|GetShowShardSideScreens_TH))))))
  (Utilities|Time|ClearandInvalidateTimerbyHandle (Variables|Mesh|GetShowShardSideScreens_TH))
  (Utilities|MeshEnabled (Variables|Mesh|GetShowShardSideScreens_MState)
    (:Enabled
      (CallFunction|ShowShardSideScreens true))
    (:Disabled
      (CallFunction|ShowShardSideScreens false))))

(event Custom|SplitSideScreens_Startup
  (Utilities|MeshEnabled (Variables|Mesh|GetSplitSideScreens_MState)
    (:Enabled
      (CallFunction|SplitSideScreens false))
    (:Default
      (bind _returnvalue (Utilities|Time|SetTimerbyEvent (delegate Custom|HandleSplitSideScreens) 0.5 true))
      (Variables|Mesh|SetSplitSideScreens_TH _returnvalue)
      (Variables|Mesh|SetSplitSideScreens_TimeeTracker (Math|DateTime|UTCNow)))))

(event Custom|HandleSplitSideScreens
  (Utilities|MeshEnabled (Variables|Mesh|GetSplitSideScreens_MState)
    (:Enabled
      _)
    (:Disabled
      _)
    (:Default
      (Utilities|OutOfTime
        (:OutOfTime
          (Utilities|Time|ClearandInvalidateTimerbyHandle (Variables|Mesh|GetSplitSideScreens_TH))))))
  (Utilities|Time|ClearandInvalidateTimerbyHandle (Variables|Mesh|GetSplitSideScreens_TH))
  (Utilities|MeshEnabled (Variables|Mesh|GetSplitSideScreens_MState)
    (:Enabled
      (CallFunction|SplitSideScreens false))))

(event Custom|MirrorSideScreens_Startup
  (Utilities|MeshEnabled (Variables|Mesh|GetMirrorSideScreens_MState)
    (:Enabled
      (CallFunction|MirrorSideScreens false))
    (:Default
      (bind _returnvalue (Utilities|Time|SetTimerbyEvent (delegate Custom|HandlleMirrorSideScreens) 0.5 true))
      (Variables|Mesh|SetMirrorSideScreens_TH _returnvalue)
      (Variables|Mesh|SetMirrorSideScreens_TimeTracker (Math|DateTime|UTCNow)))))

(event Custom|HandlleMirrorSideScreens
  (Utilities|MeshEnabled (Variables|Mesh|GetMirrorSideScreens_MState)
    (:Enabled
      _)
    (:Disabled
      _)
    (:Default
      (Utilities|OutOfTime (Variables|Mesh|GetMirrorSideScreens_TimeTracker)
        (:OutOfTime
          (Utilities|Time|ClearandInvalidateTimerbyHandle (Variables|Mesh|GetMirrorSideScreens_TH))))))
  (Utilities|Time|ClearandInvalidateTimerbyHandle (Variables|Mesh|GetMirrorSideScreens_TH))
  (Utilities|MeshEnabled (Variables|Mesh|GetMirrorSideScreens_MState)
    (:Enabled
      (CallFunction|MirrorSideScreens false))))

(event Custom|DefaultSideScreens_Startup
  (Utilities|MeshEnabled (Variables|Mesh|GetDefaultSideScreens_MState)
    (:Enabled
      (CallFunction|DefaultSideScreens false))
    (:Default
      (bind _returnvalue (Utilities|Time|SetTimerbyEvent (delegate Custom|HandleDefaultSideScreens) 0.5 true))
      (Variables|Mesh|SetDefaultSideScreens_TH _returnvalue)
      (Variables|Mesh|SetDefaultSideScreens_TimeTracker (Math|DateTime|UTCNow)))))

(event Custom|HandleDefaultSideScreens
  (Utilities|MeshEnabled (Variables|Mesh|GetDefaultSideScreens_MState)
    (:Enabled
      _)
    (:Disabled
      _)
    (:Default
      (Utilities|OutOfTime (Variables|Mesh|GetDefaultSideScreens_TimeTracker)
        (:OutOfTime
          (Utilities|Time|ClearandInvalidateTimerbyHandle (Variables|Mesh|GetDefaultSideScreens_TH))))))
  (Utilities|Time|ClearandInvalidateTimerbyHandle (Variables|Mesh|GetDefaultSideScreens_TH))
  (Utilities|MeshEnabled (Variables|Mesh|GetDefaultSideScreens_MState)
    (:Enabled
      (CallFunction|DefaultSideScreens false))))

(fn FlipAudioSubmix
  (params (First :type Boolean) (Default :type Boolean))
  (bind _mediasound (Variables|Default|GetMediaSound))
  (if (or First (Math|Boolean|NotEqual(Boolean) Default (Variables|Default|GetAudioSubmixDefault)))
    (if Default
      (Variables|Default|SetAudioSubmixDefault true)
      (Audio|Components|Audio|SetSubmixSend _mediasound "/Game/Sounds/SubMixes/Fortnite_LicensedContent_3D.Fortnite_LicensedContent_3D" 0.0 :!class /Script/AudioMixer.SynthComponent)
      (Audio|Components|Audio|SetSubmixSend _mediasound "/Game/Sounds/SubMixes/Fortnite_SFX.Fortnite_SFX" 1.0 :!class /Script/AudioMixer.SynthComponent)
      (else
        (Variables|Default|SetAudioSubmixDefault false)
        (Audio|Components|Audio|SetSubmixSend _mediasound "/Game/Sounds/SubMixes/Fortnite_SFX.Fortnite_SFX" 0.0 :!class /Script/AudioMixer.SynthComponent)
        (Audio|Components|Audio|SetSubmixSend _mediasound "/Game/Sounds/SubMixes/Fortnite_LicensedContent_3D.Fortnite_LicensedContent_3D" 1.0 :!class /Script/AudioMixer.SynthComponent)))
    (bind _outactors (Actor|GetAllActorsOfClass "/Game/Athena/Apollo/Environments/BuildingActors/Papaya/Speakers_Small_Papaya.Speakers_Small_Papaya_C"))
    (for _array_element _outactors
      (Utilities|IsValid (Variables|/Game/Athena/Apollo/Environments/BuildingActors/Papaya/Speakers_Small_Papaya.Speakers_Small_Papaya_C|GetPapayaPrototypeSpeakerComponent _array_element)
        (:Is%20Valid
          (Class|PapayaSpeakerComponent|SetSubmixSend (Variables|/Game/Athena/Apollo/Environments/BuildingActors/Papaya/Speakers_Small_Papaya.Speakers_Small_Papaya_C|GetPapayaPrototypeSpeakerComponent _array_element) Default))))))

(fn OnRep_DefaultSideScreens_MState
  (Default|MeshNode
    (:Client
      (Utilities|MeshEnabled (Variables|Mesh|GetDefaultSideScreens_MState)
        (:Enabled
          (CallFunction|DefaultSideScreens true))
        (:Disabled
          (CallFunction|DefaultSideScreens true))))))

(fn OnRep_MirrorSideScreens_MState
  (Default|MeshNode
    (:Client
      (Utilities|MeshEnabled (Variables|Mesh|GetMirrorSideScreens_MState)
        (:Enabled
          (CallFunction|MirrorSideScreens true))
        (:Disabled
          (CallFunction|DefaultSideScreens true))))))

(fn OnRep_ShowShardSideScreens_MState
  (Default|MeshNode
    (:Client
      (Utilities|MeshEnabled (Variables|Mesh|GetShowShardSideScreens_MState)
        (:Enabled
          (CallFunction|ShowShardSideScreens true))
        (:Disabled
          (CallFunction|ShowShardSideScreens false))))))

(fn OnRep_SplitSideScreens_MState
  (Default|MeshNode
    (:Client
      (Utilities|MeshEnabled (Variables|Mesh|GetSplitSideScreens_MState)
        (:Enabled
          (CallFunction|SplitSideScreens true))
        (:Disabled
          (CallFunction|DefaultSideScreens true))))))

(fn UpdateBlocks
  (bind _streamingtrigger (Variables|Default|GetStreamingTrigger))
  (bind _mediasource (Variables|/Game/Athena/BuildingActors/Props/Building/ActorBlueprints/Containers/Athena_Prop_Video_Stream_Trigger1.Athena_Prop_Video_Stream_Trigger1_C|GetMediaSource _streamingtrigger))
  (bind _returnvalue (Player|ShouldProtectPlayerfromContent _mediasource))
  (bind _returnvalue_2 (Utilities|Enum|Equal(Enum) _returnvalue "UCPBoth"))
  (bind _returnvalue_3 (not _returnvalue_2))
  (if (Variables|/Script/FortniteGame.FortStreamMediaSource|GetbShareLock _mediasource)
    (bind _returnvalue_4 (Player|IsScreenRecordinginProgress _mediasource))
    (bind _output_get (Variables|Default|SetScreenRecording _returnvalue_4))
    (if _returnvalue_4
      (Variables|Default|SetEnableVideo (not _output_get))
      (Variables|Default|SetEnableAudio (not _output_get))
      (else
        (Variables|Default|SetEnableVideo (and (not (Utilities|Enum|Equal(Enum) _returnvalue "UCPVideo")) _returnvalue_3))
        (Variables|Default|SetEnableAudio (and (not (Utilities|Enum|Equal(Enum) _returnvalue "UCPAudio")) _returnvalue_3))))
    (CallFunction|FlipAudioSubmix false (not (Utilities|String|Contains (Variables|/Script/FortniteGame.FortStreamMediaSource|GetProtectUserFromAVSettings (Variables|/Game/Athena/BuildingActors/Props/Building/ActorBlueprints/Containers/Athena_Prop_Video_Stream_Trigger1.Athena_Prop_Video_Stream_Trigger1_C|GetMediaSource (Variables|Default|GetStreamingTrigger))) "a")))))

(fn ConstructionScript
  (Development|SetHiddeninGame (Variables|StaticMeshActor|GetStaticMeshComponent) true false))
