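; Season 5 Risky Reels screen actor that shows streamed video with audio and subtitles in playlists outside the stream denylist, or in replays.
; On clients it registers with the Athena video stream trigger and follows its start, stop, audio and sharing-check events.
; For share-locked sources, content protection and screen-recording checks gate playback and pick the audio submix.
; Video toggles by material swap; audio starts/stops or switches to bus-only output.
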(event Custom|SubtitleEnable
  (params (Enable :type Boolean) (FromStart :type Boolean))
  (bind _streamingtrigger (Variables|Default|GetStreamingTrigger))
  (bind _subtitleplayer (Variables|/Game/Athena/BuildingActors/Props/Building/ActorBlueprints/Containers/Athena_Prop_Video_Stream_Trigger1.Athena_Prop_Video_Stream_Trigger1_C|GetSubtitlePlayer _streamingtrigger))
  (if (or FromStart (Media|MediaPlayer|IsPlaying (Variables|Default|GetMediaPlayer)))
    (if (not (and (Math|Boolean|Equal(Boolean) (Variables|Default|GetEnableVideo) false) (Math|Boolean|Equal(Boolean) (Variables|Default|GetEnableAudio) false)))
      (if (Stream|ShouldStreambePlaying (Variables|Default|GetPlatformSource) (Variables|StaticMeshActor|GetStaticMeshComponent) 8e+03)
        (if (not (Variables|Default|GetSubtitlesEnabled))
          (Variables|Default|SetSubtitlesEnabled true)
          (FortniteSubtitles|SubtitlesPlayer|Play _subtitleplayer))
        (else
          _))
      (else
        _))
    (if (Variables|Default|GetSubtitlesEnabled)
      (Variables|Default|SetSubtitlesEnabled false)
      (FortniteSubtitles|SubtitlesPlayer|Stop _subtitleplayer))))

(event Custom|SharingCheck
  (CallFunction|UpdateBlocks)
  (CallFunction|VideoEnable (Variables|Default|GetEnableVideo))
  (CallFunction|AudioEnable (Variables|Default|GetEnableAudio))
  (Variables|Default|SetScreenRecordingLast (Variables|Default|GetScreenRecording))
  (CallFunction|SubtitleEnable (Variables|Default|GetEnableVideo)))

(event Custom|VideoEnable
  (params (Enable :type Boolean))
  (if (Math|Boolean|NotEqual(Boolean) (Variables|Default|GetVideoEnabledLast) Enable)
    (if (Variables|/Script/FortniteGame.FortStreamMediaSource|GetbAudioOnly (Variables|Default|GetPlatformSource))
      _
      (elif Enable
        (if (not (Variables|Default|GetForceDisableVideo))
          (Variables|Default|SetVideoEnabledLast true)
          (Rendering|Material|SetMaterial (Variables|StaticMeshActor|GetStaticMeshComponent) 0 "/Game/Movies/Season5/Materials/M_RiskyReels_VideoTest_VideoPlaying.M_RiskyReels_VideoTest_VideoPlaying")
          (Variables|/Game/Athena/BuildingActors/Props/Building/ActorBlueprints/Containers/Athena_Prop_Video_Stream_Trigger1.Athena_Prop_Video_Stream_Trigger1_C|SetScreenMesh (Variables|StaticMeshActor|GetStaticMeshComponent) (Variables|Default|GetStreamingTrigger)))
        (else
          (Variables|Default|SetVideoEnabledLast false)
          (if (not (Variables|/Game/Athena/BuildingActors/Props/Building/ActorBlueprints/Containers/Athena_Prop_Video_Stream_Trigger1.Athena_Prop_Video_Stream_Trigger1_C|GetLeaveScreen (Variables|Default|GetStreamingTrigger)))
            (Rendering|Material|SetMaterial (Variables|StaticMeshActor|GetStaticMeshComponent) 0 "/Game/Environments/Props/DriveIn/Materials/M_DriveIn_Screen_Fixed.M_DriveIn_Screen_Fixed")))))))

(event Custom|AudioEnable
  (params (Enable :type Boolean) (FirstStart :type Boolean))
  (bind _mediasound (Variables|Default|GetMediaSound))
  (bind _mediasound_2 (Variables|Default|GetMediaSound))
  (if (and (not FirstStart) (== (Mesh|GetDataInt (Variables|Default|GetStreamingTrigger) (Variables|/Game/Athena/BuildingActors/Props/Building/ActorBlueprints/Containers/Athena_Prop_Video_Stream_Trigger1.Athena_Prop_Video_Stream_Trigger1_C|GetMesh_PIPVersion (Variables|Default|GetStreamingTrigger))) (Variables|Default|GetPIPVersionSinglePlayer)))
    (if (and Enable (not (Variables|Default|GetForceDisableAudio)))
      (Audio|Components|Audio|SetOutputtoBusOnly _mediasound_2 false :!class /Script/AudioMixer.SynthComponent)
      (else
        (Audio|Components|Audio|SetOutputtoBusOnly _mediasound_2 true :!class /Script/AudioMixer.SynthComponent)))
    (elif Enable
      (Media|MediaSoundComponent|SetMediaPlayer _mediasound (Variables|Default|GetMediaPlayer))
      (Components|Activation|Activate _mediasound false)
      (Synth|Components|Audio|Start _mediasound)
      (else
        (Synth|Components|Audio|Stop _mediasound)
        (if (not (Variables|Default|GetEnableVideo))
          (Components|Activation|Deactivate _mediasound))))))

(event Custom|OnVideoStopped
  (Variables|Default|SetForceDisableVideo true)
  (CallFunction|VideoEnable false)
  (Variables|Default|SetForceDisableAudio true)
  (CallFunction|AudioEnable false))

(event Custom|OnAudioEnable
  (params (Enable :type Boolean) (FirstStart :type Boolean))
  (Variables|Default|SetForceDisableAudio (not Enable))
  (CallFunction|AudioEnable Enable FirstStart))

(event Custom|On%20Video%20Started
  (params (FirstStart :type Boolean))
  (CallFunction|FlipAudioSubmix true true)
  (Variables|Default|SetForceDisableVideo false)
  (Variables|Default|SetForceDisableAudio false)
  (CallFunction|UpdateBlocks)
  (CallFunction|VideoEnable (Variables|Default|GetEnableVideo))
  (CallFunction|AudioEnable (Variables|Default|GetEnableAudio) FirstStart)
  (CallFunction|SubtitleEnable false true))

(event EventBeginPlay
  ;[1 1 1] Filter playback to Default Playlists only
  (bind _gamestate (Class|FortAsyncActionCurrentPlaylistReady|PlaylistDataReadyAsync)
    (:OnReady
      (if (or (not (GameplayTags|HasAnyTags (Class|FortGameStateAthena|GetAthenaPlaylistContextTags _gamestate) (FortGlobals|GetElectraStreamPlaylistDenylistFilter))) (Replay|GetIsPlayingReplay))
        (Networking|SwitchHasAuthority
          (:Remote
            (Development|PrintString "Create stream trigger")
            (EventDispatchers|BindEventtoApplicationHasEnteredForegroundDelegate (Variables|Default|GetApplicationLifecycle) (delegate Custom|Foreground))
            (bind _returnvalue (Actor|GetActorOfClass "/Game/Athena/BuildingActors/Props/Building/ActorBlueprints/Containers/Athena_Prop_Video_Stream_Trigger1.Athena_Prop_Video_Stream_Trigger1_C"))
            (Utilities|IsValid _returnvalue
              (:Is%20Valid
                _)
              (:Is%20Not%20Valid
                (Utilities|FlowControl|Delay 0.5)))
            (bind _output_get (Variables|Default|SetStreamingTrigger _returnvalue))
            (Variables|/Game/Athena/BuildingActors/Props/Building/ActorBlueprints/Containers/Athena_Prop_Video_Stream_Trigger1.Athena_Prop_Video_Stream_Trigger1_C|SetScreenMesh 0 _output_get)
            (Variables|/Game/Athena/BuildingActors/Props/Building/ActorBlueprints/Containers/Athena_Prop_Video_Stream_Trigger1.Athena_Prop_Video_Stream_Trigger1_C|SetUseScreenMSC true _output_get)
            (Variables|Default|SetMediaSound (Variables|/Game/Athena/BuildingActors/Props/Building/ActorBlueprints/Containers/Athena_Prop_Video_Stream_Trigger1.Athena_Prop_Video_Stream_Trigger1_C|GetScreenMediaSound _output_get))
            (Variables|/Game/Athena/BuildingActors/Props/Building/ActorBlueprints/Containers/Athena_Prop_Video_Stream_Trigger1.Athena_Prop_Video_Stream_Trigger1_C|SetScreenActor self _output_get)
            (Default|BindEventtoVideoStartedEventDispatcher _output_get (EventDispatchers|CreateEvent :!event Custom|On%20Video%20Started))
            (Default|BindEventtoVideoStoppedEventDispatcher _output_get (EventDispatchers|CreateEvent :!event Custom|OnVideoStopped))
            (Default|BindEventtoAudioStartedEventDispatcher _output_get (EventDispatchers|CreateEvent :!event Custom|AudioEnable))
            (Default|BindEventtoVideoSharingCheck _output_get (EventDispatchers|CreateEvent :!event Custom|SharingCheck))))))))

(event Custom|Foreground
  (CallFunction|SharingCheck))

(event EventEndPlay
  (params (EndPlayReason :type EEndPlayReason%20Enum :default "Destroyed"))
  (CallFunction|SubtitleEnable false)
  (EventDispatchers|UnbindallEventsfromApplicationHasEnteredForegroundDelegate (Variables|Default|GetApplicationLifecycle)))

(fn CheckBlocks
  (params (Video :type Boolean))
  (if Video
    (return (or (Utilities|Enum|Equal(Enum) (Player|ShouldProtectPlayerfromContent (Variables|/Game/Athena/BuildingActors/Props/Building/ActorBlueprints/Containers/Athena_Prop_Video_Stream_Trigger1.Athena_Prop_Video_Stream_Trigger1_C|GetMediaSource (Variables|Default|GetStreamingTrigger))) "UCPAudio") (Utilities|Enum|Equal(Enum) (Player|ShouldProtectPlayerfromContent (Variables|/Game/Athena/BuildingActors/Props/Building/ActorBlueprints/Containers/Athena_Prop_Video_Stream_Trigger1.Athena_Prop_Video_Stream_Trigger1_C|GetMediaSource (Variables|Default|GetStreamingTrigger))) "UCPAudio")))
    (else
      (return (or (Utilities|Enum|Equal(Enum) (Player|ShouldProtectPlayerfromContent (Variables|/Game/Athena/BuildingActors/Props/Building/ActorBlueprints/Containers/Athena_Prop_Video_Stream_Trigger1.Athena_Prop_Video_Stream_Trigger1_C|GetMediaSource (Variables|Default|GetStreamingTrigger))) "UCPAudio") (Utilities|Enum|Equal(Enum) (Player|ShouldProtectPlayerfromContent (Variables|/Game/Athena/BuildingActors/Props/Building/ActorBlueprints/Containers/Athena_Prop_Video_Stream_Trigger1.Athena_Prop_Video_Stream_Trigger1_C|GetMediaSource (Variables|Default|GetStreamingTrigger))) "UCPAudio"))))))

(fn FlipAudioSubmix
  (params (First :type Boolean) (Default :type Boolean))
  (bind _mediasound (Variables|Default|GetMediaSound))
  (if (or First (Math|Boolean|NotEqual(Boolean) Default (Variables|Default|GetAudioSubmixDefault)))
    (if Default
      (Variables|Default|SetAudioSubmixDefault true)
      (Audio|Components|Audio|SetSubmixSend _mediasound "/Game/Sounds/SubMixes/Fortnite_LicensedContent_3D.Fortnite_LicensedContent_3D" 0.0 :!class /Script/AudioMixer.SynthComponent)
      (Audio|Components|Audio|SetSubmixSend _mediasound "/Game/Sounds/SubMixes/Fortnite_SFX.Fortnite_SFX" 1.0 :!class /Script/AudioMixer.SynthComponent)
      (else
        (Variables|Default|SetAudioSubmixDefault false)
        (Audio|Components|Audio|SetSubmixSend _mediasound "/Game/Sounds/SubMixes/Fortnite_SFX.Fortnite_SFX" 0.0 :!class /Script/AudioMixer.SynthComponent)
        (Audio|Components|Audio|SetSubmixSend _mediasound "/Game/Sounds/SubMixes/Fortnite_LicensedContent_3D.Fortnite_LicensedContent_3D" 1.0 :!class /Script/AudioMixer.SynthComponent)))))

(fn UpdateBlocks
  (bind _streamingtrigger (Variables|Default|GetStreamingTrigger))
  (bind _mediasource (Variables|/Game/Athena/BuildingActors/Props/Building/ActorBlueprints/Containers/Athena_Prop_Video_Stream_Trigger1.Athena_Prop_Video_Stream_Trigger1_C|GetMediaSource _streamingtrigger))
  (bind _returnvalue (Player|ShouldProtectPlayerfromContent _mediasource))
  (bind _returnvalue_2 (Utilities|Enum|Equal(Enum) _returnvalue "UCPAudio"))
  (bind _returnvalue_3 (not _returnvalue_2))
  (if (Variables|/Script/FortniteGame.FortStreamMediaSource|GetbShareLock _mediasource)
    (bind _returnvalue_4 (Player|IsScreenRecordinginProgress _mediasource))
    (bind _output_get (Variables|Default|SetScreenRecording _returnvalue_4))
    (if _returnvalue_4
      (Variables|Default|SetEnableVideo (not _output_get))
      (Variables|Default|SetEnableAudio (not _output_get))
      (else
        (Variables|Default|SetEnableVideo (and (not (Utilities|Enum|Equal(Enum) _returnvalue "UCPAudio")) _returnvalue_3))
        (Variables|Default|SetEnableAudio (and (not (Utilities|Enum|Equal(Enum) _returnvalue "UCPAudio")) _returnvalue_3))))
    (CallFunction|FlipAudioSubmix false (not (Utilities|String|Contains (Variables|/Script/FortniteGame.FortStreamMediaSource|GetProtectUserFromAVSettings (Variables|/Game/Athena/BuildingActors/Props/Building/ActorBlueprints/Containers/Athena_Prop_Video_Stream_Trigger1.Athena_Prop_Video_Stream_Trigger1_C|GetMediaSource (Variables|Default|GetStreamingTrigger))) "a")))))

(fn ConstructionScript)
