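; Replicated trigger that drives synchronized, region-specific video playback on named screens across the mesh network.
; Root events for warming, play/stop, seek, loop, PiP/mirror, bandwidth, CDN weights and priority filter by ScreenName or "all", then write per-region values.
; RepNotifies relay state: Edge converts play UTC timestamps into server-time offsets, and Clients fire dispatchers for playback.
; Parallel non-region RepNotifies handle the same state without RegionData.
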(event EventBeginPlay
  (switch Utilities|FlowControl|Switch|SwitchonEMeshNetworkNodeType (MeshNetworkSubsystem|GetMeshNetworkNodeType (GameInstanceSubsystems|GetMeshNetworkSubsystem))
    (:Client
      (Development|PrintString (Utilities|String|Append "VideoMeshTrigger Created for ScreenName: " (Variables|Default|GetScreenName)) false))))

(event Custom|RootStartWarming
  (params (ScreenName :type String) (Enable :type Boolean) (Simple :type Boolean) (CDNWeights :type String) (RegionData :type String))
  (bind _returnvalue (Math|DateTime|UTCNow))
  (bind _stoputctime (Variables|Default|GetStopUTCTime))
  (CallFunction|PrintIncomingData ScreenName "None" "RootStartWarming")
  (bind _thisscreen (CallFunction|CheckScreen ScreenName))
  (if _thisscreen
    (Class|FortMediaRegionLockComponent|MeshAddRegionValueString CDNWeights (Variables|RegionData|GetRegion_WarmingCDNWeights) RegionData)
    (if Enable
      (Class|FortMediaRegionLockComponent|MeshAddRegionValueBool Simple (Variables|RegionData|GetRegion_SimpleWarming) RegionData)
      (Class|FortMediaRegionLockComponent|MeshAddRegionValueDateTime _returnvalue (Variables|RegionData|GetRegion_WarmingStartUTCTime_GS) RegionData)
      (Variables|RegionData|SetRegion_StartWarmingUTCTime (Utilities|Struct|MakeMeshRegionLockDataDateTime _returnvalue RegionData))
      (else
        (Class|FortMediaRegionLockComponent|MeshAddRegionValueDateTime _stoputctime (Variables|RegionData|GetRegion_WarmingStartUTCTime_GS) RegionData)
        (Class|FortMediaRegionLockComponent|MeshAddRegionValueBool false (Variables|RegionData|GetRegion_SimpleWarming) RegionData)
        (Variables|RegionData|SetRegion_StartWarmingUTCTime (Utilities|Struct|MakeMeshRegionLockDataDateTime _stoputctime RegionData))))))

(event Custom|RootPlayVideo
  (params (ScreenName :type String) (VideoUID :type String) (PiP :type Integer) (Live :type Boolean) (Distance :type Integer) (RegionData :type String))
  (bind _returnvalue (Math|DateTime|UTCNow))
  (bind _pipversionoff (Variables|Default|GetPIPVersionOff))
  (CallFunction|PrintIncomingData ScreenName VideoUID "RootPlayVideo")
  (bind _thisscreen (CallFunction|CheckScreen ScreenName))
  (if _thisscreen
    (Class|FortMediaRegionLockComponent|MeshAddRegionValueString VideoUID (Variables|RegionData|GetRegion_VideoUID) RegionData)
    (if (!= PiP _pipversionoff)
      (CallFunction|PIPAlsoVersion PiP Live VideoUID true RegionData)
      (else
        (Class|FortMediaRegionLockComponent|MeshAddRegionValueInt _pipversionoff (Variables|RegionData|GetRegion_PIPVersion) RegionData)))
    (Class|FortMediaRegionLockComponent|MeshAddRegionValueBool Live (Variables|RegionData|GetRegion_IsLive) RegionData)
    (Class|FortMediaRegionLockComponent|MeshAddRegionValueInt Distance (Variables|RegionData|GetRegion_ScreenDistance) RegionData)
    (Class|FortMediaRegionLockComponent|MeshAddRegionValueDateTime _returnvalue (Variables|RegionData|GetRegion_PlayUTCTime_GS) RegionData)
    (Variables|RegionData|SetRegion_PlayUTCTime (Utilities|Struct|MakeMeshRegionLockDataDateTime _returnvalue RegionData))))

(event Custom|RootStopVideo
  (params (ScreenName :type String) (LeaveScreen :type Boolean) (RegionData :type String))
  (bind _stoputctime (Variables|Default|GetStopUTCTime))
  (CallFunction|PrintIncomingData ScreenName "None" "RootStopVideo")
  (bind _thisscreen (CallFunction|CheckScreen ScreenName))
  (if _thisscreen
    (Class|FortMediaRegionLockComponent|MeshRemoveRegionValueString (Variables|RegionData|GetRegion_VideoUID) RegionData)
    (Class|FortMediaRegionLockComponent|MeshAddRegionValueBool LeaveScreen (Variables|RegionData|GetRegion_LeaveScreen) RegionData)
    (Class|FortMediaRegionLockComponent|MeshRemoveRegionValueBool (Variables|RegionData|GetRegion_LoopVideo_GS) RegionData)
    (Variables|RegionData|SetRegion_LoopVideo (Utilities|Struct|MakeMeshRegionLockDataBool false RegionData))
    (Class|FortMediaRegionLockComponent|MeshRemoveRegionValueBool (Variables|RegionData|GetRegion_IsLive) RegionData)
    (Class|FortMediaRegionLockComponent|MeshRemoveRegionValueInt (Variables|RegionData|GetRegion_ScreenDistance) RegionData)
    (bind _outitem (Class|FortMediaRegionLockComponent|MeshGetRegionValueBool RegionData (Variables|RegionData|GetRegion_PlaywithPiP)))
    (if (Utilities|Struct|BreakMeshRegionLockDataBool _outitem)
      (bind _outitem_2 (Class|FortMediaRegionLockComponent|MeshGetRegionValueInt RegionData (Variables|RegionData|GetRegion_PIPVersion)))
      (CallFunction|PIPAlsoVersion (Utilities|Struct|BreakMeshRegionLockDataInt _outitem_2) false "" false RegionData)
      (else
        _))
    (Class|FortMediaRegionLockComponent|MeshRemoveRegionValueDateTime (Variables|RegionData|GetRegion_PlayUTCTime_GS) RegionData)
    (Class|FortMediaRegionLockComponent|MeshRemoveRegionValueDateTime (Variables|RegionData|GetRegion_RootUtcTime) RegionData)
    (Variables|RegionData|SetRegion_PlayUTCTime (Utilities|Struct|MakeMeshRegionLockDataDateTime _stoputctime RegionData))
    (Class|FortMediaRegionLockComponent|MeshRemoveRegionValueDateTime (Variables|RegionData|GetRegion_WarmingStartUTCTime_GS) RegionData)
    (Variables|RegionData|SetRegion_StartWarmingUTCTime (Utilities|Struct|MakeMeshRegionLockDataDateTime _stoputctime RegionData))))

(event Custom|RootEnableLoopVideo
  (params (ScreenName :type String) (RegionData :type String))
  (CallFunction|PrintIncomingData ScreenName "None" "RootEnableLoopVideo")
  (bind _thisscreen (CallFunction|CheckScreen ScreenName))
  (if _thisscreen
    (Class|FortMediaRegionLockComponent|MeshAddRegionValueBool true (Variables|RegionData|GetRegion_LoopVideo_GS) RegionData)
    (Variables|RegionData|SetRegion_LoopVideo (Utilities|Struct|MakeMeshRegionLockDataBool true RegionData))))

(event Custom|RootDisableLoopVideo
  (params (ScreenName :type String) (RegionData :type String))
  (CallFunction|PrintIncomingData ScreenName "None" "RootDisableLoopVideo")
  (bind _thisscreen (CallFunction|CheckScreen ScreenName))
  (if _thisscreen
    (Class|FortMediaRegionLockComponent|MeshRemoveRegionValueBool (Variables|RegionData|GetRegion_LoopVideo_GS) RegionData)
    (Variables|RegionData|SetRegion_LoopVideo (Utilities|Struct|MakeMeshRegionLockDataBool false RegionData))))

(event Custom|RootSetMaxBW
  (params (ScreenName :type String) (MaxBW :type Integer) (RegionData :type String))
  (CallFunction|PrintIncomingData ScreenName "None" "RootSetMaxBW")
  (bind _thisscreen (CallFunction|CheckScreen ScreenName))
  (if _thisscreen
    (Class|FortMediaRegionLockComponent|MeshAddRegionValueInt MaxBW (Variables|RegionData|GetRegion_MaxBandwidth_GS) RegionData)
    (Variables|RegionData|SetRegion_MaxBandwidth (Utilities|Struct|MakeMeshRegionLockDataInt MaxBW RegionData))))

(event Custom|RootSeekVideo
  (params (ScreenName :type String) (Time :type Float%20%28double-precision%29) (Exact :type Boolean) (RegionData :type String))
  (CallFunction|PrintIncomingData ScreenName "None" "RootSeekVideo")
  (bind _thisscreen (CallFunction|CheckScreen ScreenName))
  (if _thisscreen
    (Class|FortMediaRegionLockComponent|MeshAddRegionValueFloat Time (Variables|RegionData|GetRegion_SeekTime) RegionData)
    (Utilities|MeshNode
      (:Client
        (Default|CallOnSeekPlayback (Variables|Default|GetScreenName) Time Exact RegionData)))))

(event Custom|RootSetScreenMirror
  (params (ScreenName :type String) (Enable :type Boolean) (RegionData :type String))
  (bind _region_pipmirrorenable (Variables|RegionData|GetRegion_PiPMirrorEnable))
  (CallFunction|PrintIncomingData ScreenName "None" "RootSetScreenMirror")
  (bind _thisscreen (CallFunction|CheckScreen ScreenName))
  (if _thisscreen
    (if Enable
      (Class|FortMediaRegionLockComponent|MeshAddRegionValueBool Enable _region_pipmirrorenable RegionData)
      (bind (_outitem _returnvalue) (Class|FortMediaRegionLockComponent|MeshGetRegionValueInt RegionData (Variables|RegionData|GetRegion_PIPVersion)))
      (if (or (not _returnvalue) (== (Utilities|Struct|BreakMeshRegionLockDataInt _outitem) (Variables|Default|GetPIPVersionOff)))
        (Class|FortMediaRegionLockComponent|MeshAddRegionValueInt 1 (Variables|RegionData|GetRegion_PIPVersion) RegionData)
        (else
          _))
      (Utilities|MeshNode
        (:Client
          (Default|CallOnSetMirrorScreen (Variables|Default|GetScreenName) Enable RegionData)))
      (else
        (Class|FortMediaRegionLockComponent|MeshRemoveRegionValueBool _region_pipmirrorenable RegionData)
        (Class|FortMediaRegionLockComponent|MeshRemoveRegionValueInt (Variables|RegionData|GetRegion_PIPVersion) RegionData)))))

(event Custom|RootStartPiP
  (params (ScreenName :type String) (VideoUID :type String) (Enable :type Boolean) (Live :type Boolean) (RegionData :type String))
  (bind _returnvalue (Math|DateTime|UTCNow))
  (CallFunction|PrintIncomingData ScreenName VideoUID "RootStartPiP")
  (bind _thisscreen (CallFunction|CheckScreen ScreenName))
  (if _thisscreen
    (if Enable
      (Class|FortMediaRegionLockComponent|MeshAddRegionValueBool Live (Variables|RegionData|GetRegion_IsPIPLive) RegionData)
      (Class|FortMediaRegionLockComponent|MeshAddRegionValueString VideoUID (Variables|RegionData|GetRegion_PiPVideoUID) RegionData)
      (Class|FortMediaRegionLockComponent|MeshAddRegionValueBool true (Variables|RegionData|GetRegion_PiPEnable) RegionData)
      (else
        (Class|FortMediaRegionLockComponent|MeshRemoveRegionValueBool (Variables|RegionData|GetRegion_IsPIPLive) RegionData)
        (Class|FortMediaRegionLockComponent|MeshRemoveRegionValueInt (Variables|RegionData|GetRegion_PIPVersion) RegionData)
        (Class|FortMediaRegionLockComponent|MeshRemoveRegionValueBool (Variables|RegionData|GetRegion_PiPEnable) RegionData)
        (Class|FortMediaRegionLockComponent|MeshRemoveRegionValueString (Variables|RegionData|GetRegion_PiPVideoUID) RegionData)))
    (Class|FortMediaRegionLockComponent|MeshAddRegionValueDateTime _returnvalue (Variables|RegionData|GetRegion_PiPPlayUTCTime_GS) RegionData)
    (Variables|RegionData|SetRegion_PiPPlayUTCTime (Utilities|Struct|MakeMeshRegionLockDataDateTime _returnvalue RegionData))))

(event Custom|RootForcePiP
  (params (ScreenName :type String) (Counter :type Integer64) (RegionData :type String))
  (CallFunction|PrintIncomingData ScreenName "none" "RootForcePiP")
  (bind _thisscreen (CallFunction|CheckScreen ScreenName))
  (if _thisscreen
    (if (Math|Integer64|NotEqual(Integer64) Counter -1)
      (Class|FortMediaRegionLockComponent|MeshAddRegionValueInt64 Counter (Variables|RegionData|GetRegion_PiPForceCounter) RegionData)
      (else
        (Class|FortMediaRegionLockComponent|MeshRemoveRegionValueInt64 (Variables|RegionData|GetRegion_PiPForceCounter) RegionData)))
    (Utilities|MeshNode
      (:Client
        (Default|CallOnForcePiP ScreenName Counter RegionData)))))

(event Custom|RootSetCDNWeights
  (params (ScreenName :type String) (Weights :type String) (PIPToo :type Boolean) (Clear :type Boolean) (RegionData :type String))
  (CallFunction|PrintIncomingData ScreenName "none" "RootSetCDNWeights")
  (bind _thisscreen (CallFunction|CheckScreen ScreenName))
  (if _thisscreen
    (if Clear
      (Class|FortMediaRegionLockComponent|MeshRemoveRegionValueString (Variables|RegionData|GetRegion_CDNWeights) RegionData)
      (if PIPToo
        (Class|FortMediaRegionLockComponent|MeshRemoveRegionValueBool (Variables|RegionData|GetRegion_PIPCDNDistroSet) RegionData)
        (else
          _))
      (Utilities|MeshNode
        (:Client
          (Default|CallOnCDNWeights "" RegionData)))
      (else
        (Class|FortMediaRegionLockComponent|MeshAddRegionValueString Weights (Variables|RegionData|GetRegion_CDNWeights) RegionData)
        (if PIPToo
          (Class|FortMediaRegionLockComponent|MeshAddRegionValueBool true (Variables|RegionData|GetRegion_PIPCDNDistroSet) RegionData)
          (else
            _))
        (Utilities|MeshNode
          (:Client
            (Default|CallOnCDNWeights Weights RegionData)))))))

(event Custom|RootSetScreenPrio
  (params (ScreenName :type String) (Distance :type Integer) (RegionData :type String))
  (CallFunction|PrintIncomingData ScreenName "none" "RootSetScreenPrio")
  (bind _thisscreen (CallFunction|CheckScreen ScreenName))
  (if _thisscreen
    (Class|FortMediaRegionLockComponent|MeshAddRegionValueInt Distance (Variables|RegionData|GetRegion_ScreenDistance) RegionData)
    (Utilities|MeshNode
      (:Client
        (Default|CallOnScreenPrio Distance RegionData)))))

(event Custom|RootClearAllData
  (params (ScreeName :type String))
  (CallFunction|PrintIncomingData ScreeName "none" "RootClearAllData")
  (bind _thisscreen (CallFunction|CheckScreen ScreeName))
  (if _thisscreen
    (Utilities|MeshNode
      (:Edge
        (Utilities|Array|Clear (Variables|RegionData|GetRegion_PlayServerTimeSeconds_GS))
        (Utilities|Array|Clear (Variables|RegionData|GetRegion_PiPPlayerServerTimeSeconds_GS)))
      (:Root
        (Utilities|FlowControl|Sequence
          (:then_0
            (Utilities|Array|Clear (Variables|RegionData|GetRegion_VideoPlayStartTime))
            (Utilities|Array|Clear (Variables|RegionData|GetRegion_RootUtcTime))
            (Utilities|Array|Clear (Variables|RegionData|GetRegion_LeaveScreen))
            (Utilities|Array|Clear (Variables|RegionData|GetRegion_IsLive))
            (Utilities|Array|Clear (Variables|RegionData|GetRegion_VideoUID))
            (Utilities|Array|Clear (Variables|RegionData|GetRegion_ScreenDistance))
            (Utilities|Array|Clear (Variables|RegionData|GetRegion_WarmingCDNWeights)))
          (:then_1
            (Utilities|Array|Clear (Variables|RegionData|GetRegion_SimpleWarming))
            (Utilities|Array|Clear (Variables|RegionData|GetRegion_VideoWarmingStartTime))
            (Utilities|Array|Clear (Variables|RegionData|GetRegion_PlaywithPiP))
            (Utilities|Array|Clear (Variables|RegionData|GetRegion_SeekTime))
            (Utilities|Array|Clear (Variables|RegionData|GetRegion_PIPVersion))
            (Utilities|Array|Clear (Variables|RegionData|GetRegion_PiPMirrorEnable))
            (Utilities|Array|Clear (Variables|RegionData|GetRegion_PiPEnable)))
          (:then_2
            (Utilities|Array|Clear (Variables|RegionData|GetRegion_IsPIPLive))
            (Utilities|Array|Clear (Variables|RegionData|GetRegion_PiPForceCounter))
            (Utilities|Array|Clear (Variables|RegionData|GetRegion_PIPCDNDistroSet))
            (Utilities|Array|Clear (Variables|RegionData|GetRegion_CDNWeights))
            (Utilities|Array|Clear (Variables|RegionData|GetRegion_PiPVideoUID))
            (Utilities|Array|Clear (Variables|RegionData|GetRegion_PlayUTCTime_GS))
            (Utilities|Array|Clear (Variables|RegionData|GetRegion_WarmingStartUTCTime_GS)))
          (:then_3
            (Utilities|Array|Clear (Variables|RegionData|GetRegion_PlayServerTimeSeconds_GS))
            (Utilities|Array|Clear (Variables|RegionData|GetRegion_StartWarmingServerTimeSeconds_GS))
            (Utilities|Array|Clear (Variables|RegionData|GetRegion_LoopVideo_GS))
            (Utilities|Array|Clear (Variables|RegionData|GetRegion_MaxBandwidth_GS))
            (Utilities|Array|Clear (Variables|RegionData|GetRegion_PiPPlayUTCTime_GS))
            (Utilities|Array|Clear (Variables|RegionData|GetRegion_PiPPlayerServerTimeSeconds_GS)))
          (:then_4
            (Variables|RegionData|SetRegion_PlayUTCTime)
            (Variables|RegionData|SetRegion_PlayServerTimeSeconds)
            (Variables|RegionData|SetRegion_StartWarmingUTCTime)
            (Variables|RegionData|SetRegion_StartWarmingServerTimeSeconds)
            (Variables|RegionData|SetRegion_LoopVideo)
            (Variables|RegionData|SetRegion_MaxBandwidth))
          (:then_5
            (Variables|RegionData|SetRegion_PiPPlayUTCTime)
            (Variables|RegionData|SetRegion_PiPPlayerServerTimeSeconds)))))))

(fn CheckScreen
  (params (ScreenName :type String))
  (return (or (Utilities|String|EqualExactly(String) ScreenName "all") (Utilities|String|EqualExactly(String) ScreenName (Variables|Default|GetScreenName)))))

(fn OnRep_LoopVideo
  (bind _screenname (Variables|Default|GetScreenName))
  (if (Variables|Default|GetLoopVideo)
    (Default|CallOnEnableLoopVideo _screenname)
    (else
      (Default|CallOnDisableLoopVideo _screenname))))

(fn OnRep_MaxBandwidth
  (Networking|SwitchHasAuthority
    (:Remote
      (Default|CallOnSetBandwidthLimit (Variables|Default|GetScreenName) (Variables|Default|GetMaxBandwidth)))))

(fn OnRep_PiPPlayerServerTimeSeconds
  (Networking|SwitchHasAuthority
    (:Remote
      (Default|CallOnStartPiP (Variables|Default|GetScreenName) (Variables|Default|GetPiPVideoUID) (Variables|Default|GetPiPEnable) (Variables|Default|GetIsPIPLive)))))

(fn OnRep_PiPPlayUTCTIme
  (switch Utilities|FlowControl|Switch|SwitchonEMeshNetworkNodeType (MeshNetworkSubsystem|GetMeshNetworkNodeType (GameInstanceSubsystems|GetMeshNetworkSubsystem))
    (:Edge
      (Variables|Default|SetPiPPlayerServerTimeSeconds (- (Math|Timespan|GetTotalSeconds (Math|DateTime|DateTime-DateTime (Math|DateTime|UTCNow) (Variables|Default|GetPiPPlayUTCTIme))) (Fortnite|Sync|Time|GetServerWorldTimeSeconds))))))

(fn OnRep_PlayServerTimeSeconds
  (Networking|SwitchHasAuthority
    (:Remote
      (if (== (Variables|Default|GetPlayServerTimeSeconds) -1.0)
        (Variables|Default|SetVideoPlayStartTime -1.0)
        (Default|CallOnStopPlayback (Variables|Default|GetScreenName) (Variables|Default|GetLeaveScreen))
        (else
          (LocalVariables|SetCachedServerWorldTimeSeconds (Fortnite|Sync|Time|GetServerWorldTimeSeconds))
          (Fortnite|FortShippingLog 0 (Utilities|String|Append "PlayServerTimeSeconds replicated as: " (Utilities|String|ToString(Float) (Variables|Default|GetPlayServerTimeSeconds)) "
" "Client GetServerWorldTimeSeconds:" (Utilities|String|ToString(Float) (LocalVariables|GetCachedServerWorldTimeSeconds))))
          (bind _output_get (Variables|Default|SetVideoPlayStartTime (- (LocalVariables|GetCachedServerWorldTimeSeconds) (Variables|Default|GetPlayServerTimeSeconds))))
          (Default|CallOnReadytoPlay _output_get (Variables|Default|GetVideoUID) (Variables|Default|GetScreenName) (Variables|Default|GetIsLive)))))))

(fn OnRep_PlayUTCTime
  (switch Utilities|FlowControl|Switch|SwitchonEMeshNetworkNodeType (MeshNetworkSubsystem|GetMeshNetworkNodeType (GameInstanceSubsystems|GetMeshNetworkSubsystem))
    (:Edge
      (if (Math|DateTime|Equal(DateTime) (Variables|Default|GetPlayUTCTime) (Variables|Default|GetStopUTCTime))
        (Variables|Default|SetPlayServerTimeSeconds -1.0)
        (else
          (bind _output_get (Variables|Default|SetRootUtcTime (Math|DateTime|UTCNow)))
          (Variables|Default|SetPlayServerTimeSeconds (- (Math|Timespan|GetTotalSeconds (Math|DateTime|DateTime-DateTime _output_get (Variables|Default|GetPlayUTCTime))) (Fortnite|Sync|Time|GetServerWorldTimeSeconds))))))))

(fn OnRep_Region_LoopVideo
  (bind _region_loopvideo (Variables|RegionData|GetRegion_LoopVideo))
  (bind (_data _regiondata) (Utilities|Struct|BreakMeshRegionLockDataBool _region_loopvideo))
  (bind _screenname (Variables|Default|GetScreenName))
  (Utilities|MeshNode
    (:Client
      (if _data
        (Default|CallOnEnableLoopVideo _screenname _regiondata)
        (else
          (Default|CallOnDisableLoopVideo _screenname _regiondata))))))

(fn OnRep_Region_MaxBandwidth
  (bind (_data _regiondata) (Utilities|Struct|BreakMeshRegionLockDataInt (Variables|RegionData|GetRegion_MaxBandwidth)))
  (Utilities|MeshNode
    (:Client
      (Default|CallOnSetBandwidthLimit (Variables|Default|GetScreenName) _data _regiondata))))

(fn OnRep_Region_PiPPlayerServerTimeSeconds
  (bind _region_pipplayerservertimeseconds (Variables|RegionData|GetRegion_PiPPlayerServerTimeSeconds))
  (bind (_data _regiondata) (Utilities|Struct|BreakMeshRegionLockDataFloat _region_pipplayerservertimeseconds))
  (Utilities|MeshNode
    (:Client
      (bind _outitem (Class|FortMediaRegionLockComponent|MeshGetRegionValueBool _regiondata (Variables|RegionData|GetRegion_IsPIPLive)))
      (bind _outitem_2 (Class|FortMediaRegionLockComponent|MeshGetRegionValueBool _regiondata (Variables|RegionData|GetRegion_PiPEnable)))
      (bind _outitem_3 (Class|FortMediaRegionLockComponent|MeshGetRegionValueString _regiondata (Variables|RegionData|GetRegion_PiPVideoUID)))
      (Default|CallOnStartPiP (Variables|Default|GetScreenName) (Utilities|Struct|BreakMeshRegionLockDataString _outitem_3) (Utilities|Struct|BreakMeshRegionLockDataBool _outitem_2) (Utilities|Struct|BreakMeshRegionLockDataBool _outitem) _regiondata))))

(fn OnRep_Region_PiPPlayUTCTime
  (bind _region_pipplayutctime (Variables|RegionData|GetRegion_PiPPlayUTCTime))
  (bind (_data _regiondata) (Utilities|Struct|BreakMeshRegionLockDataDateTime _region_pipplayutctime))
  (bind _returnvalue (Fortnite|Sync|Time|GetServerWorldTimeSeconds))
  (bind _returnvalue_2 (Math|DateTime|UTCNow))
  (bind _returnvalue_3 (Math|DateTime|DateTime-DateTime _returnvalue_2 _data))
  (bind _returnvalue_4 (Math|Timespan|GetTotalSeconds _returnvalue_3))
  (bind _returnvalue_5 (- _returnvalue_4 _returnvalue))
  (Utilities|MeshNode
    (:Edge
      (Class|FortMediaRegionLockComponent|MeshAddRegionValueFloat _returnvalue_5 (Variables|RegionData|GetRegion_PiPPlayerServerTimeSeconds_GS) _regiondata)
      (Variables|RegionData|SetRegion_PiPPlayerServerTimeSeconds (Utilities|Struct|MakeMeshRegionLockDataFloat _returnvalue_5 _regiondata)))))

(fn OnRep_Region_PlayServerTimeSeconds
  (bind _region_playservertimeseconds (Variables|RegionData|GetRegion_PlayServerTimeSeconds))
  (bind (_data _regiondata) (Utilities|Struct|BreakMeshRegionLockDataFloat _region_playservertimeseconds))
  (if (== _data -1.0)
    (Utilities|MeshNode
      (:Client
        (bind _outitem (Class|FortMediaRegionLockComponent|MeshGetRegionValueBool _regiondata (Variables|RegionData|GetRegion_LeaveScreen)))
        (Default|CallOnStopPlayback (Variables|Default|GetScreenName) (Utilities|Struct|BreakMeshRegionLockDataBool _outitem) _regiondata)))
    (else
      (Utilities|MeshNode
        (:Client
          (bind _outitem_2 (Class|FortMediaRegionLockComponent|MeshGetRegionValueBool _regiondata (Variables|RegionData|GetRegion_IsLive)))
          (bind _outitem_3 (Class|FortMediaRegionLockComponent|MeshGetRegionValueString _regiondata (Variables|RegionData|GetRegion_VideoUID)))
          (Default|CallOnReadytoPlay _data (Utilities|Struct|BreakMeshRegionLockDataString _outitem_3) (Variables|Default|GetScreenName) (Utilities|Struct|BreakMeshRegionLockDataBool _outitem_2) _regiondata))))))

(fn OnRep_Region_PlayUTCTime
  (bind _region_playutctime (Variables|RegionData|GetRegion_PlayUTCTime))
  (bind (_data _regiondata) (Utilities|Struct|BreakMeshRegionLockDataDateTime _region_playutctime))
  (bind _returnvalue (Fortnite|Sync|Time|GetServerWorldTimeSeconds))
  (bind _returnvalue_2 (Math|DateTime|DateTime-DateTime (Math|DateTime|UTCNow) _data))
  (bind _returnvalue_3 (Math|Timespan|GetTotalSeconds _returnvalue_2))
  (bind _returnvalue_4 (- _returnvalue_3 _returnvalue))
  (bind _returnvalue_5 (Math|DateTime|UTCNow))
  (Utilities|MeshNode
    (:Edge
      (if (Math|DateTime|Equal(DateTime) (Variables|Default|GetStopUTCTime) _data)
        (Class|FortMediaRegionLockComponent|MeshAddRegionValueFloat -1.0 (Variables|RegionData|GetRegion_PlayServerTimeSeconds_GS) _regiondata)
        (Variables|RegionData|SetRegion_PlayServerTimeSeconds (Utilities|Struct|MakeMeshRegionLockDataFloat -1.0 _regiondata))
        (else
          (Class|FortMediaRegionLockComponent|MeshAddRegionValueDateTime _returnvalue_5 (Variables|RegionData|GetRegion_RootUtcTime) _regiondata)
          (Class|FortMediaRegionLockComponent|MeshAddRegionValueFloat _returnvalue_4 (Variables|RegionData|GetRegion_PlayServerTimeSeconds_GS) _regiondata)
          (Variables|RegionData|SetRegion_PlayServerTimeSeconds (Utilities|Struct|MakeMeshRegionLockDataFloat _returnvalue_4 _regiondata)))))))

(fn OnRep_Region_ScreenDistance)

(fn OnRep_Region_StartWarmingServerTimeSeconds
  (bind _region_startwarmingservertimeseconds (Variables|RegionData|GetRegion_StartWarmingServerTimeSeconds))
  (bind (_data _regiondata) (Utilities|Struct|BreakMeshRegionLockDataFloat _region_startwarmingservertimeseconds))
  (if (== _data -1.0)
    (Utilities|MeshNode
      (:Client
        (bind _outitem (Class|FortMediaRegionLockComponent|MeshGetRegionValueBool _regiondata (Variables|RegionData|GetRegion_SimpleWarming)))
        (Default|CallOnStopWarming (Variables|Default|GetScreenName) (Utilities|Struct|BreakMeshRegionLockDataBool _outitem) _regiondata)))
    (else
      (Utilities|MeshNode
        (:Client
          (bind _outitem_2 (Class|FortMediaRegionLockComponent|MeshGetRegionValueBool _regiondata (Variables|RegionData|GetRegion_SimpleWarming)))
          (Default|CallOnStartWarming _data (Variables|Default|GetScreenName) (Utilities|Struct|BreakMeshRegionLockDataBool _outitem_2) _regiondata))))))

(fn OnRep_Region_StartWarmingUTCTime
  (bind _region_startwarmingutctime (Variables|RegionData|GetRegion_StartWarmingUTCTime))
  (bind (_data _regiondata) (Utilities|Struct|BreakMeshRegionLockDataDateTime _region_startwarmingutctime))
  (if (Math|DateTime|Equal(DateTime) _data (Variables|Default|GetStopUTCTime))
    (Variables|RegionData|SetRegion_StartWarmingServerTimeSeconds (Utilities|Struct|MakeMeshRegionLockDataFloat -1.0 _regiondata))
    (else
      (bind _outitem (Class|FortMediaRegionLockComponent|MeshGetRegionValueDateTime _regiondata (Variables|RegionData|GetRegion_RootUtcTime)))
      (Variables|RegionData|SetRegion_StartWarmingServerTimeSeconds (Utilities|Struct|MakeMeshRegionLockDataFloat (- (Fortnite|Sync|Time|GetServerWorldTimeSeconds) (Math|Timespan|GetTotalSeconds (Math|DateTime|DateTime-DateTime (Utilities|Struct|BreakMeshRegionLockDataDateTime _outitem) _data))) _regiondata)))))

(fn OnRep_ScreenDistance
  (Default|CallOnScreenPrio (Variables|Default|GetScreenDistance)))

(fn OnRep_ScreenName)

(fn OnRep_StartWarmingServerTimeSeconds
  (Networking|SwitchHasAuthority
    (:Remote
      (if (== (Variables|Default|GetStartWarmingServerTimeSeconds) -1.0)
        (Variables|Default|SetVideoWarmingStartTime -1.0)
        (Default|CallOnStopWarming (Variables|Default|GetScreenName))
        (else
          (LocalVariables|SetCachedServerWorldTimeSeconds (Fortnite|Sync|Time|GetServerWorldTimeSeconds))
          (Fortnite|FortShippingLog 0 (Utilities|String|Append "GALILEO: StartWarmingServerTimeSeconds replicated as: " (Utilities|String|ToString(Float) (Variables|Default|GetStartWarmingServerTimeSeconds)) "
" "Client GetServerWorldTimeSeconds:" (Utilities|String|ToString(Float) (LocalVariables|GetCachedServerWorldTimeSeconds))))
          (bind _output_get (Variables|Default|SetVideoWarmingStartTime (- (LocalVariables|GetCachedServerWorldTimeSeconds) (Variables|Default|GetStartWarmingServerTimeSeconds))))
          (Default|CallOnStartWarming _output_get (Variables|Default|GetScreenName)))))))

(fn OnRep_StartWarmingUTCTime
  (switch Utilities|FlowControl|Switch|SwitchonEMeshNetworkNodeType (MeshNetworkSubsystem|GetMeshNetworkNodeType (GameInstanceSubsystems|GetMeshNetworkSubsystem))
    (:Edge
      (if (Math|DateTime|Equal(DateTime) (Variables|Default|GetStartWarmingUTCTime) (Variables|Default|GetStopUTCTime))
        (Variables|Default|SetStartWarmingServerTimeSeconds -1.0)
        (else
          (Variables|Default|SetStartWarmingServerTimeSeconds (- (Fortnite|Sync|Time|GetServerWorldTimeSeconds) (Math|Timespan|GetTotalSeconds (Math|DateTime|DateTime-DateTime (Variables|Default|GetRootUtcTime) (Variables|Default|GetStartWarmingUTCTime))))))))))

(fn PIPAlsoVersion
  (params (Version :type Integer) (Live :type Boolean) (VUID :type String) (Enable :type Boolean) (RegionData :type String))
  (bind _screenname (Variables|Default|GetScreenName))
  (bind _region_playwithpip (Variables|RegionData|GetRegion_PlaywithPiP))
  (bind (_outitem _returnvalue) (Class|FortMediaRegionLockComponent|MeshGetRegionValueInt RegionData (Variables|RegionData|GetRegion_PIPVersion)))
  (if (or (not _returnvalue) (== (Utilities|Struct|BreakMeshRegionLockDataInt _outitem) (Variables|Default|GetPIPVersionOff)))
    (Class|FortMediaRegionLockComponent|MeshAddRegionValueInt Version (Variables|RegionData|GetRegion_PIPVersion) RegionData)
    (else
      _))
  (if (== Version 0)
    (CallFunction|RootStartPiP _screenname VUID Enable Live RegionData)
    (if Enable
      (Class|FortMediaRegionLockComponent|MeshAddRegionValueBool Enable _region_playwithpip RegionData)
      (else
        (Class|FortMediaRegionLockComponent|MeshRemoveRegionValueBool _region_playwithpip RegionData)))
    (return)
    (elif (or (== Version 1) (== Version 2))
      (CallFunction|RootSetScreenMirror _screenname Enable RegionData)
      (if Enable
        _
        (else
          _))
      (else
        _))))

(fn PrintIncomingData
  (params (ScreenName :type String) (VideoUID :type String) (Command :type String))
  (Development|PrintString (Utilities|String|Append Command " - ScreenName: " ScreenName " - VideoUID: " VideoUID) false))

(fn ConstructionScript)
