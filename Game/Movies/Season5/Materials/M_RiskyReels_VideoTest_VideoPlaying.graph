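# Unlit Risky Reels drive-in screen test material for showing video or an animated intermission card.
# A static switch, AllowVideoPlayback, selects the media player texture through an inverse-tonemap function or the intermission composite.
# The intermission steps through three flipbook phases on scaled time, animating UV scale, rotation and offset, using the flipbook as its own blend mask over a background.
# Separate branches provide movie brightness, tinted screen diffuse and a normal map.
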Material /Game/Movies/Season5/Materials/M_RiskyReels_VideoTest_VideoPlaying
graph {
 properties {pixelDepthOffsetMode=PDOM_Legacy, shadingModel=MSM_Unlit, usedShadingModels=Unlit}
 nodes [
  Constant3Vector()
  Multiply(
   A=ScalarParameter() {parameterName=BaseColor}
   B=TextureSample() {texture=/Game/Environments/Props/DriveIn/Textures/T_DriveIn_Screen_Fixed_D}.RGB
  )
  n1 = MaterialFunctionCall!(UVs, Texture=TextureObject() {texture=/Game/Movies/Season5/RiskyReels_MediaPlayer_Video}) {materialFunction=/Game/Movies/InversePostProcess/MF_invertFNBRtonemap}
  Multiply(A=@n1.Result, B=ScalarParameter() {defaultValue=3, parameterName=MovieBrightness})
  n2 = Multiply(A=Time(), B=ScalarParameter() {defaultValue=0.2, group=Intermission-Carousel, parameterName=TimeScale})
  n3 = Floor(None=Fmod(A=@n2, B=Constant() {r=3}))
  n4 = Fmod(A=@n2, B=Constant() {r=1})
  n5 = Sine(None=@n4) {period=2}
  n6 = MaterialFunctionCall(UVs=TextureCoordinate(), Texture Scale=LinearInterpolate(Alpha=Power(Base=@n5) {constExponent=0.2}) {constA=0.25, constB=0.95}) {materialFunction=&ScaleUVsByCenter}
  n7 = LinearInterpolate(Alpha=Clamp(None=@n4)) {constA=-0.02, constB=0.02}
  n8 = LinearInterpolate(
   A=LinearInterpolate(Alpha=@n4) {constA=-0.1, constB=0.1}
   B=LinearInterpolate(Alpha=@n4) {constA=-0.05, constB=0.05}
   Alpha=Sine(None=@n4) {period=2}
  )
  n9 = ScalarParameter() {defaultValue=-0.05, group=Intermission-Carousel, parameterName=Y-GlobalOffset}
  n10 = Multiply(A=Add(A=Rotator(Coordinate=@n6.UVs, Time=@n7) {centerY=0.75, speed=6.2832}, B=AppendVector(A=@n8, B=@n9)), B=Constant2Vector() {g=1.2, r=2})
  n11 = Constant2Vector() {g=-0.1048, r=-0.45}
  n12 = MaterialFunctionCall(
   "Animation  Phase (0-1)"=Multiply(A=Add(A=@n3)) {constB=0.25}
   Texture=TextureObject() {texture=/Game/Movies/Season5/Textures/T_Intermission_Flipbook}
   UVs=Saturate(None=Add(A=@n10, B=@n11))
  ) {materialFunction=&FlipBook}
  StaticSwitchParameter(
   True=@n1.Result
   False=LinearInterpolate(A=TextureSample() {texture=/Game/Movies/Season5/Textures/T-IntermissionBackground}.RGB, B=@n12.Result, Alpha=@n12.Result)
  ) {defaultValue=true, parameterName=AllowVideoPlayback}
  TextureSample() {samplerType=SAMPLERTYPE_Normal, texture=/Game/Environments/Props/DriveIn/Textures/T_DriveIn_Screen_Fixed_N}
 ]
}
